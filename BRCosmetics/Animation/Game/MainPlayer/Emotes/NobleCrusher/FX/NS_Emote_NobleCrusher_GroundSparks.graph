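# Niagara system for the Noble Crusher emote that throws hot, glowing sparks off the ground from a sampled socket.
# Core layers socket-following sprites, meshes, a flickering light and a decal-like glow by spawn group.
# Sparks emits two weighted variants with gravity, drag, collision and speed-based stretching.
# Owner speed gates spawning and light, User.Multiplier scales intensity, and the Classic_Sparks and Core001 emitters are disabled.
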NiagaraSystem /BRCosmetics/Animation/Game/MainPlayer/Emotes/NobleCrusher/FX/NS_Emote_NobleCrusher_GroundSparks
dependencies {
 usedDataInterfaces [
  /Script/Niagara.NiagaraDataInterfaceArrayDistributionInt
  &NiagaraDataInterfaceArrayFloat
  &NiagaraDataInterfaceCurve
  &NiagaraDataInterfaceColorCurve
  &NiagaraDataInterfaceSocketReader
  &NiagaraDataInterfaceVector2DCurve
  &NiagaraDataInterfaceSpriteRendererInfo
  &NiagaraDataInterfaceCamera
 ]
 usedDynamicInputs [&Abs_Float]
 usedModules [
  &self:SystemSpawnScript.SystemScriptSource.SystemScriptGraph.NiagaraNodeAssignment_1.SetVariables_#1
  &self:SystemSpawnScript.SystemScriptSource.SystemScriptGraph.NiagaraNodeAssignment_2.SetVariables_#2
  &SystemState
  &BlendSpaceHelper
  &self:SystemSpawnScript.SystemScriptSource.SystemScriptGraph.NiagaraNodeAssignment_0.SetVariables_#3
  &Timeline
 ]
}
emitters {
 Classic_Sparks {
  bEnabled=false
  emitterUpdateScript [
   &EmitterState(Life Cycle Mode=::NewEnumerator1, Loop Delay=2.3833, Loop Duration=2, UseLoopDelay=false)
   &SpawnRate(
    Spawn Probability=&Multiply_Float(
     A=&MakeCustomFloatFromBool(
      Bool=&SetBoolByFloatComparison(A=&VectorLength(Vector=$Engine.Owner.Velocity), B=5, Comparison Type=::NewEnumerator6)
      True Float=0
      False Float=1
     )
     B=&RemapRange(Input Value To Remap=$User.Multiplier, Input Min=0, Input Max=1, Output Min=0.25, Output Max=1)
    )
    SpawnRate=40
    Use Spawn Probability=true
   )
  ]
  particleSpawnScript [
   /SetVariables(Particles.ParticleRandom:NiagaraFloat=&SelectFloatFromArray(
     Direct Array Index=/Niagara/DynamicInputs/Arrays/SelectIntFromWeightedDistributionArray(
      Array Sampling Mode:ENiagara_LifetimeMode=::NewEnumerator1
      Direct Array Index:NiagaraInt32=?
      Direct Array Mode:ENiagara_WrapClamp=::NewEnumerator0
      Evaluation Type:ENiagaraRandomnessEvaluation=::NewEnumerator0
      Recalculate Random Each Loop:NiagaraBool=false
      Weighted Distribution Int Array:NiagaraDataInterfaceArrayDistributionInt=DataInterface() {propertyValues {ArrayData [{value=0, weight=0.75}, {value=1, weight=0.25}]}}
      Randomness Mode:ENiagaraRandomnessMode=::NewEnumerator0
      Random Seed:NiagaraInt32=0
      Fixed Random Seed:NiagaraBool=false
     )
     Float Selection Array=DataInterface() {propertyValues {FloatData [0,1]}}
    )) {isSetParameters=true, module=SetVariables_#4}
   &InitializeParticle(
    Color Mode=::NewEnumerator0
    Lifetime=&MergeHelper_FloatSwitch(
     0=&RandomRangeFloat(Minimum=0.18, Maximum=0.25)
     1=&RandomRangeFloat(Minimum=0.4, Maximum=0.6)
     2=?
     3=?
     Index=&MakeIntFromFloat(Float=$Particles.ParticleRandom, Conversion Mode=::NewEnumerator1)
     Num=::NewEnumerator0
    )
    Mesh Scale Max=<0.0663 0.0663 0.0663>
    Mesh Scale Min=<2.4885 2.8084 2.6334>
    Mesh Scale Mode=::NewEnumerator2
    Position Mode=::NewEnumerator0
    Sprite Rotation Mode=::NewEnumerator2
    Sprite Size Max=<3 3>
    Sprite Size Min=<1 1>
    Sprite Size Mode=::NewEnumerator2
    Write Material Random=true
   )
   &SocketLocation(
    "Inherit Velocity (Scale)"=-2
    SamplePosOffset=true
    Sampled Orientation Coordinate Space:ENiagaraCoordinateSpace=::Local
    Sampled Position Offset=<0 0 -1>
    Sampled Position Offset Coordinate Space:ENiagaraSkelMeshTransforms=::NewEnumerator2
    Sampled Velocity Coordinate Space:ENiagaraCoordinateSpace=::Simulation
    Socket Reader=$System.Socket Reader
    Socket Sampling Mode=::NewEnumerator3
    VelocitySpeedLimit=false
   )
   &AddVelocity(
    Cone Angle=&MergeHelper_FloatSwitch(
     0=40
     1=30
     2=?
     3=?
     Index=&MakeIntFromFloat(Float=$Particles.ParticleRandom, Conversion Mode=::NewEnumerator1)
     Num=::NewEnumerator0
    )
    Cone Axis=<-0.5 0.45 0.15>
    Module.UseConeFalloff=false
    Module.UseVelDistribution=false
    Override Seed=false
    Velocity Mode=::NewEnumerator2
    Velocity Speed=&Multiply_Float(
     A=&MergeHelper_FloatSwitch(
      0=&RandomRangeFloat(Minimum=50, Maximum=100)
      1=&RandomRangeFloat(Minimum=75, Maximum=150)
      2=?
      3=?
      Index=&MakeIntFromFloat(Float=$Particles.ParticleRandom, Conversion Mode=::NewEnumerator1)
      Num=::NewEnumerator0
     )
     B=&RemapRange(Input Value To Remap=$User.Multiplier, Input Min=0, Input Max=1, Output Min=0.5, Output Max=1.1)
    )
   ) {enabled=false, module=AddVelocity001}
   &AddVelocity(
    Cone Angle=&MergeHelper_FloatSwitch(
     0=40
     1=50
     2=?
     3=?
     Index=&MakeIntFromFloat(Float=$Particles.ParticleRandom, Conversion Mode=::NewEnumerator1)
     Num=::NewEnumerator0
    )
    Cone Axis=<-1 0.3 -0.2>
    Module.UseConeFalloff=false
    Module.UseVelDistribution=false
    Override Seed=false
    Velocity Mode=::NewEnumerator2
    Velocity Speed=&Multiply_Float(
     A=&MergeHelper_FloatSwitch(
      0=&RandomRangeFloat(Minimum=50, Maximum=100)
      1=&RandomRangeFloat(Minimum=75, Maximum=150)
      2=?
      3=?
      Index=&MakeIntFromFloat(Float=$Particles.ParticleRandom, Conversion Mode=::NewEnumerator1)
      Num=::NewEnumerator0
     )
     B=&RemapRange(Input Value To Remap=$User.Multiplier, Input Min=0, Input Max=1, Output Min=1.2, Output Max=1.7)
    )
   ) {module=AddVelocity002}
   &AddVelocity(
    Cone Angle=&MergeHelper_FloatSwitch(
     0=40
     1=30
     2=?
     3=?
     Index=&MakeIntFromFloat(Float=$Particles.ParticleRandom, Conversion Mode=::NewEnumerator1)
     Num=::NewEnumerator0
    )
    Cone Axis=&RotateVectorByQuaternion(Vector To Rotate=<-0.5 0.45 0.15>, Quaternion=$Output.SocketLocation.SampledRotation)
    Module.UseConeFalloff=false
    Module.UseVelDistribution=false
    Override Seed=false
    Velocity Mode=::NewEnumerator2
    Velocity Speed=&Multiply_Float(
     A=&MergeHelper_FloatSwitch(
      0=&RandomRangeFloat(Minimum=50, Maximum=150)
      1=&RandomRangeFloat(Minimum=75, Maximum=200)
      2=?
      3=?
      Index=&MakeIntFromFloat(Float=$Particles.ParticleRandom, Conversion Mode=::NewEnumerator1)
      Num=::NewEnumerator0
     )
     B=&RemapRange(Input Value To Remap=$User.Multiplier, Input Min=0, Input Max=1, Output Min=0.65, Output Max=1.35)
    )
   ) {enabled=false}
   &InheritVelocity(Inherited Velocity Amount Scale=<500 500 500>, Inherited Velocity Speed Limit=1500, Speed Limit=true, UseSpeedThreshold=false) {enabled=false}
  ]
  particleUpdateScript [
   &ParticleState()
   /SetVariables(
    Particles.Flicker:NiagaraFloat=&Waveform("Waveform [1]"=::NewEnumerator0, "[1] Frequency"=75, "[1] Amplitude Min/Max"=<0 2>)
    Particles.Flicker Size:NiagaraFloat=&Waveform("Waveform [1]"=::NewEnumerator0, "[1] Frequency"=75, "[1] Amplitude Min/Max"=<0.5 1>)
    Particles.Flicker Value:NiagaraFloat=&Waveform("Waveform [1]"=::NewEnumerator0, "[1] Frequency"=75, "[1] Amplitude Min/Max"=<2 8>)
   ) {isSetParameters=true, module=SetVariables_#5}
   &ScaleSpriteSize(
    Scale Sprite Size Mode=::NewEnumerator0
    Uniform Scale Factor=&Lerp_Float(
     A=1
     B=&Multiply_Float(A=$Particles.Flicker Size, B=1.5)
     Alpha=&FloatFromCurve(FloatCurve=DataInterface() {propertyValues {Curve {keys [<at 0 0.1076>,<at 0.2485 0.6128>,<at 0.45 1>,<at 1 1>]}}})
    )
   )
   &ScaleColor(
    Scale RGB=&MakeVectorFromLinearColorRGB(Color=&ColorFromCurve(ColorCurve=DataInterface() {propertyValues {
        AlphaCurve {keys [<at 0 1>]}
        BlueCurve {keys [<at 0 0.2604>,<at 0.1639 0>,<at 1 0>]}
        GreenCurve {keys [<at 0 1.0503>,<at 0.1639 3.75>,<at 1 0.25>]}
        RedCurve {keys [<at 0 5>,<at 0.1639 15>,<at 1 1>]}
       }}))
    ScaleA=true
    ScaleRGB=true
   )
   &ScaleColor(
    Scale Mode=::NewEnumerator1
    Scale RGBA=&Multiply_Vector4(
     A=&MakeVector4_FromLinearColor(LinearColor=&ColorFromCurve(ColorCurve=DataInterface() {propertyValues {
         AlphaCurve {keys [<at 0.122 1>]}
         BlueCurve {keys [<at 0 1.0417>,<at 0.1355 0.9911>,<at 0.3032 0>,<at 1 0>]}
         GreenCurve {keys [<at 0 7.4334>,<at 0.1355 10.4768>,<at 0.3032 9.1355>,<at 1 3.6718>]}
         RedCurve {keys [<at 0 20>,<at 0.1355 16>,<at 0.3032 13>,<at 1 9>]}
        }}))
     B=&MakeVector4_FromLinearColor(LinearColor=&MakeLinearColorFromVectorAndFloat(
       "Vector (RGB)"=&Add_Vector(
        A=<1 1 1>
        B=/Niagara/DynamicInputs/TypeConversions/MakeVectorFromLinearColorHSV(Color:LinearColor=/Niagara/DynamicInputs/LinearColor/MakeLinearColorFromHSV(
          HSV From:NiagaraInt32=1
          HSV:Vector3f=?
          Hue:NiagaraFloat=0
          Saturation:NiagaraFloat=0
          Value:NiagaraFloat=&Lerp_Float(
           A=0
           B=$Particles.Flicker Value
           Alpha=&FloatFromCurve(FloatCurve=DataInterface() {propertyValues {Curve {keys [<at 0 0>,<at 0.5 0>,<at 0.55 0.9624>,<at 1 1>]}}})
          )
          Alpha:NiagaraFloat=1
         ))
       )
       "Float (Alpha)"=&Lerp_Float(
        A=1
        B=$Particles.Flicker
        Alpha=&FloatFromCurve(FloatCurve=DataInterface() {propertyValues {Curve {keys [<at 0 0>,<at 0.25 0>,<at 0.35 1>,<at 1 1>]}}})
       )
      ))
    )
    ScaleRGBA=true
   ) {enabled=false, module=ScaleColor001}
   &GravityForce(Gravity=&MergeHelper_VectorSwitch(
     0=&RandomRangeVector(Minimum=<0 0 -500>, Maximum=<0 0 -500>)
     1=<0 0 -1200>
     2=?
     3=?
     Index=&MakeIntFromFloat(Float=$Particles.ParticleRandom, Conversion Mode=::NewEnumerator1)
     Num=::NewEnumerator0
    ))
   &Collision(
    Advanced Aging Rate=1
    Age Colliding Particles=false
    Collision Enabled=&SetBoolByFloatComparison(A=$Particles.NormalizedAge, B=0.2)
    Enable Rest State=false
    Friction=0.6
    Friction During a Bounce=0.4
    Kill On Collision:NiagaraInt32=&RandomRangeInt(Minimum=3, Maximum=5)
    KillOnCollision=true
    Max Number Of Collisions=5
    Module.EnableMaxCollisionCount=false
    Module.Randomize Collision Normal Vector=false
   ) {enabled=false}
   &Collision(
    Advanced Aging Rate=1
    Age Colliding Particles=false
    Collision Enabled=&SetBoolByFloatComparison(A=$Particles.NormalizedAge, B=0.2)
    Enable Rest State=false
    Friction=0.6
    Friction During a Bounce=0.4
    Kill On Collision:NiagaraInt32=&RandomRangeInt(Minimum=3, Maximum=5)
    KillOnCollision=false
    Max Number Of Collisions=5
    Module.EnableMaxCollisionCount=false
    Module.Randomize Collision Normal Vector=false
    Restitution=0
   ) {module=Collision001}
   &Drag(Drag=&RandomRangeFloat(Minimum=4.5, Maximum=7), Use Linear Drag=true, Use Rotational Drag=false)
   &InheritVelocity(Inherited Velocity Amount Scale=<-1 -1 -1>, Speed Limit=true, UseSpeedThreshold=false) {enabled=false, module=InheritVelocity001}
   &SolveForcesAndVelocity(Clamp Velocity=false, Limit Acceleration=false)
   &ScaleSpriteSizeBySpeed(
    Max Scale Factor=&RandomRangeVector2D(Minimum=<0.5 2>, Maximum=<4 5>)
    Min Scale Factor=&MergeHelper_Vector2Switch(
     0=<0.4 0.8>
     1=<0.4 1>
     2=?
     3=?
     4=?
     Index=&MakeIntFromFloat(Float=$Particles.ParticleRandom, Conversion Mode=::NewEnumerator1)
     Num=::NewEnumerator0
    )
    Velocity Threshold=1500
   ) {enabled=false}
   &ScaleSpriteSizeBySpeed(Max Scale Factor=<0.5 3>, Min Scale Factor=<0.5 0.5>, Velocity Threshold=125) {module=ScaleSpriteSizeBySpeed001}
   /SetVariables(Particles.Color=&MultiplyLinearColor(
     Linear Color=&LerpLinearColors(
      StartColor=<rgba 25 8.3333 0 1>
      EndColor=<rgba 25 2.5 0 1>
      LerpFactor=&Multiply_Float(A=$Particles.NormalizedAge, B=1.5)
     )
     Scale Factor=&MakeVector4_FromFloat(Float=1)
    )) {isSetParameters=true, module=SetVariables_#6}
   &BlendSpace(
    "Position Blend (World <-> Local)":NiagaraFloat=&MakeCustomFloatFromBool(Bool=$Particles.HasCollided, True Float=0, False Float=0.1, Translation Contribution:NiagaraFloat=1, Rotation Contribution:NiagaraFloat=1)
    Position Coordinate Space=1
   ) {module=BlendSpace001}
  ]
  rendererClasses [&NiagaraSpriteRendererProperties]
  renderers [{rendererClass=&NiagaraSpriteRendererProperties}]
 }
 Core {
  emitterUpdateScript [
   &EmitterState(Inactive Response=::NewEnumerator2, Life Cycle Mode=::NewEnumerator1, Loop Duration=4, UseLoopDelay=false)
   &SpawnRate(
    Spawn Probability=&MakeCustomFloatFromBool(
     Bool=&SetBoolByFloatComparison(A=&VectorLength(Vector=$Engine.Owner.Velocity), B=5, Comparison Type=::NewEnumerator6)
     True Float=0
     False Float=1
    )
    SpawnRate=4
    Use Spawn Probability=true
   ) {module=SpawnRate001}
   &SpawnRate(
    Spawn Probability=&MakeCustomFloatFromBool(
     Bool=&SetBoolByFloatComparison(A=&VectorLength(Vector=$Engine.Owner.Velocity), B=5, Comparison Type=::NewEnumerator6)
     True Float=0
     False Float=1
    )
    SpawnGroup=1
    SpawnRate=40
    Use Spawn Probability=true
   )
   &SpawnBurst_Instantaneous(Spawn Count=2, Spawn Group=2, Use Spawn Probability=false, UseLoopCountLimit=false) {module=SpawnBurst_Instantaneous001}
   &SpawnBurst_Instantaneous(Spawn Group=3, Use Spawn Probability=false, UseLoopCountLimit=false)
  ]
  particleSpawnScript [
   /SetVariables(Particles.VisibilityTag=&SpawnGroup()) {isSetParameters=true, module=SetVariables_#7}
   &InitializeParticle(
    Color=<rgba 1.5 0.8757 0.4297 1>
    Lifetime=&MergeHelper_FloatSwitch(
     0=0.36
     1=&UniformRangedFloat(Minimum=0.05, Maximum=0.1)
     2=4
     3=4
     Index=$Particles.VisibilityTag
     Num=::NewEnumerator2
    )
    Mesh Scale Max=&MakeVector(
     X=&RemapRange(Input Value To Remap=$User.Multiplier, Input Min=0, Input Max=1, Output Min=0.4, Output Max=0.5)
     Y=0.2
     Z=0.35
    )
    Mesh Scale Min=&MakeVector(
     X=&RemapRange(Input Value To Remap=$User.Multiplier, Input Min=0, Input Max=1, Output Min=0.3, Output Max=0.35)
     Y=0.2
     Z=0.1
    )
    Mesh Scale Mode=::NewEnumerator2
    Position Mode=::NewEnumerator1
    Position Offset=&MaskVectorBySpawnGroup(Masked Vector=<0 0 0>, Passthrough Vector=<0 0 0>)
    Sprite Size=&MergeHelper_Vector2Switch(0=<130 72>, 1=<0 0>, 2=<15 15>, 3=?, 4=?, Index=$Particles.VisibilityTag, Num=::NewEnumerator1)
    Sprite Size Mode=::NewEnumerator4
    UsePositionOffset=false
    Write Material Random=true
   ) {module=InitializeParticle001}
   &DynamicMaterialParameters(
    Index 0 Param 1=0.2
    Module.Param0WriteEnabled=true
    Module.Param1WriteEnabled=true
    Module.Param2WriteEnabled=true
    Module.Param3WriteEnabled=true
   )
   &CameraOffset(Camera Offset Amount=&MergeHelper_FloatSwitch(0=10, 1=0, 2=5, 3=?, Index=$Particles.VisibilityTag, Num=::NewEnumerator1)) {module=CameraOffset001}
   &SocketLocation(
    "Inherit Velocity (Scale)"=0.01
    Orientation Sampling=::NewEnumerator1
    Position Sampling=::NewEnumerator1
    SamplePosOffset=true
    Sampled Orientation Coordinate Space:ENiagaraCoordinateSpace=::Local
    Sampled Position Offset=<0 0 -1.5>
    Sampled Velocity Coordinate Space:ENiagaraCoordinateSpace=::Local
    Socket Reader=$System.Socket Reader
    Socket Sampling Mode=::NewEnumerator3
    Velocity Sampling:ENiagara_AttributeSamplingApplyOutput=::NewEnumerator0
    VelocitySpeedLimit=false
   ) {module=SocketLocation001}
   /SetVariables(Particles.PositionGround_Spark:NiagaraPosition=$Output.SocketLocation001.SampledPosition) {isSetParameters=true, module=SetVariables_#8}
   &InitialMeshOrientation(
    Mesh Orientation Mode=::NewEnumerator2
    Orientation Coordinate Space=::Local
    Orientation Vector=<0 -2 -2>
    Rotation=&RandomRangeVector(Minimum=<0.45 -0.1 0>, Maximum=<0.5 -0.1 0>)
    Rotation Coordinate Space=::NewEnumerator0
    Use Rotation Vector=true
   ) {module=InitialMeshOrientation001}
   /SetVariables(Particles.Color_Decal:LinearColor=<rgba 18 5.954 0.7914 0.3652>, Particles.DynamicMaterialParameter_Decal:Vector4f=<1 0 1 0>) {isSetParameters=true, module=SetVariables_#9}
   &SpriteFacingAndAlignment(
    Sprite Alignment=<0 0 1>
    Sprite Facing=&MergeHelper_VectorSwitch(0=<0 0 0>, 1=<0 0 0>, 2=<0 1 0>, 3=?, Index=$Particles.VisibilityTag, Num=::NewEnumerator1)
    WriteAlignment=false
    WriteFacing=true
   )
   &SpriteFacingAndAlignment(Sprite Alignment=<1 1 0>, Sprite Facing=<0 1 0>, WriteAlignment=false, WriteFacing=true) {enabled=false, module=SpriteFacingAndAlignment001}
  ]
  particleUpdateScript [
   &ParticleState()
   &ScaleSpriteSize(
    Scale Sprite Size Mode=::NewEnumerator0
    Uniform Scale Factor=&RemapRange(Input Value To Remap=$User.Multiplier, Input Min=0, Input Max=1, Output Min=0.25, Output Max=1)
   )
   &SolveForcesAndVelocity(Clamp Velocity=false, Limit Acceleration=false)
   &SocketLocation(
    Position Sampling=::NewEnumerator1
    SamplePosOffset=true
    Sampled Position Offset=&MergeHelper_VectorSwitch(0=<0 0 0>, 1=<0 0 0>, 2=<0 0 -2>, 3=?, Index=$Particles.VisibilityTag, Num=::NewEnumerator1)
    Socket Reader=$System.Socket Reader
    Socket Sampling Mode=::NewEnumerator3
    VelocitySpeedLimit=false
   )
   /SetVariables(
    Particles.SparkColor:LinearColor=&MakeLinearColorFromVectorAndFloat(
     "Vector (RGB)"=&MakeVectorFromLinearColorRGB(Color=<rgba 5 2.5329 0.0658 1>)
     "Float (Alpha)"=&MakeCustomFloatFromBool(
      Bool=&SetBoolByFloatComparison(A=&VectorLength(Vector=$Engine.Owner.Velocity), B=5, Comparison Type=::NewEnumerator6)
      True Float=0
      False Float=1
     )
    )
    Particles.SparkSize:Vector2f=&MakeVector2D(
     X=500
     Y=&ClampFloat(Float=&Sine(Normalized Angle=$Emitter.Age, Period=0.0616, Scale=4.1323, Bias=4.3498), Min=12.5264, Max=24.152)
    )
    Particles.Vector 2D:Vector2f=&Multiply_Vector2D_ByFloat(
     Float=&RemapRange(Input Value To Remap=$User.Multiplier, Input Min=0, Input Max=1, Output Min=0.6, Output Max=1)
     Vector2D=&Vector2DFromCurve(
      Vector2Curve=DataInterface() {propertyValues {
        XCurve {defaultValue=1, keys [{tangentMode=RCTM_User, time=0, value=0.5}, {tangentMode=RCTM_User, time=1, value=1}]}
        YCurve {defaultValue=1, keys [{tangentMode=RCTM_User, time=0, value=0.5}, {tangentMode=RCTM_User, time=1, value=1}]}
       }}
      ScaleCurve=&RandomRangeVector2D(Minimum=<23.4 8.7388>, Maximum=<23.4 8.3315>)
     )
    )
   ) {isSetParameters=true, module=SetVariables_#10}
   /SetVariables(
    Particles.SparkColor:LinearColor=&MakeLinearColorFromVectorAndFloat(
     "Vector (RGB)"=&MakeVectorFromLinearColorRGB(Color=<rgba 5 2.5329 0.0658 1>)
     "Float (Alpha)"=&Lerp_Float(A=0, B=1, Alpha=$System.Timeline.Time)
    )
    Particles.SparkSize:Vector2f=&MakeVector2D(
     X=500
     Y=&ClampFloat(Float=&Sine(Normalized Angle=$Emitter.Age, Period=0.0616, Scale=4.1323, Bias=4.3498), Min=12.5264, Max=24.152)
    )
    Particles.Vector 2D:Vector2f=&Multiply_Vector2D_ByFloat(
     Float=&RemapRange(Input Value To Remap=$User.Multiplier, Input Min=0, Input Max=1, Output Min=0.6, Output Max=1)
     Vector2D=&Vector2DFromCurve(
      Vector2Curve=DataInterface() {propertyValues {
        XCurve {defaultValue=1, keys [{tangentMode=RCTM_User, time=0, value=0.5}, {tangentMode=RCTM_User, time=1, value=1}]}
        YCurve {defaultValue=1, keys [{tangentMode=RCTM_User, time=0, value=0.5}, {tangentMode=RCTM_User, time=1, value=1}]}
       }}
      ScaleCurve=&RandomRangeVector2D(Minimum=<23.4 8.7388>, Maximum=<23.4 8.3315>)
     )
    )
   ) {enabled=false, isSetParameters=true}
   &SubUVAnimation(
    Random Start Frame=true
    "Sprite Renderer "=DataInterface() {propertyValues {SpriteRenderer=&self:Core_1.NiagaraSpriteRendererProperties_1}}
    SubUV Lookup Index Scale=5.7547
    UseEndFrame=false
    UseStartFrame=false
   )
   /SetVariables(Particles.Position=&ConvertPositionToVector(
     Input Position=&AddVectorToPosition(
      Position=$Output.SocketLocation.SampledPosition
      Vector=&MergeHelper_VectorSwitch(0=<0 1 -2>, 1=<0 0 0>, 2=<0 2 -3>, 3=?, Index=$Particles.VisibilityTag, Num=::NewEnumerator1)
     )
     Conversion Operation:ENiagara_LWCConvertPosToVec=::NewEnumerator0
    )) {isSetParameters=true, module=SetVariables_#11}
   &FN_Light_Attributes(
    Color=<rgba 2 1.0278 0 1>
    Create Color Attribute=true
    Create Exponent Attribute=true
    Create Radius Attribute=true
    Exponent=5
    Radius=40
    Scale Color=&Multiply_Float(
     A=1.3
     B=&MakeCustomFloatFromBool(
      Bool=&SetBoolByFloatComparison(A=&VectorLength(Vector=$Engine.Owner.Velocity), B=5)
      True Float=&Multiply_Float(
       A=&RemapRange(Input Value To Remap=$User.Multiplier, Input Min=0, Input Max=1, Output Min=0.6, Output Max=1.5)
       B=$System.FlickerSpeed
      )
      False Float=0
     )
    )
   )
   /SetVariables(Particles.PositionLights:NiagaraPosition=&AddVectorToPosition(Position=$Output.SocketLocation.SampledPosition, Vector=<-10 -2 -5>)) {isSetParameters=true, module=SetVariables_#12}
   /SetVariables(Particles.Color_GroundSpark:LinearColor=&MultiplyLinearColor(
     Linear Color=&LerpLinearColors(
      StartColor=<rgba 25 8.3333 0 1>
      EndColor=<rgba 25 2.0833 0 1>
      LerpFactor=&Multiply_Float(A=$Particles.NormalizedAge, B=1)
     )
     Scale Factor=&MakeVector4_FromFloat(Float=&RemapRange(Input Value To Remap=$User.Multiplier, Input Min=0, Input Max=1, Output Min=0.3, Output Max=0.5))
    )) {isSetParameters=true, module=SetVariables_#13}
   &ScaleMeshSize(Scale Factor=&VectorFromFloat(Value=&FloatFromCurve(
      FloatCurve=DataInterface() {propertyValues {Curve {keys [{arriveTangent=2.2457, interpMode=RCIM_Cubic, leaveTangent=2.2457, tangentMode=RCTM_User, time=0, value=0}, <at 1 1>]}}}
      Scale Curve=&RemapRange(Input Value To Remap=$User.Multiplier, Input Min=0, Input Max=1, Output Min=0.4, Output Max=0.55)
     )))
   /SetVariables(Particles.Color_Decal_Update:LinearColor=&MultiplyLinearColor(
     Linear Color=$Particles.Initial.Color_Decal
     Scale Factor=&MakeVector4FromVectorAndFloat(
      "Vector (RGB)"=<1 1 1>
      "Float (Alpha)"=&Multiply_Float(
       A=&Multiply_Float(
        A=1.25
        B=&RemapRange(
         Input Value To Remap=&DotProduct(
          Vector A=&ReturnCameraVector(Camera Vectors=::NewEnumerator4, Coordinate Space=::Simulation, Origin Position=$Particles.Position)
          Vector B=$Particles.SpriteFacing
         )
         Input Min=0
         Input Max=0.5
         Output Min=0.35
         Output Max=1
        )
       )
       B=&Multiply_Float(
        A=&Multiply_Float(
         A=&RemapRange(Input Value To Remap=$User.Multiplier, Input Min=0, Input Max=1, Output Min=0.4, Output Max=1)
         B=&Lerp_Float(A=0, B=1, Alpha=$System.Timeline.Time)
        )
        B=&Waveform("Waveform [1]"=::NewEnumerator0, "[1] Frequency"=4.1057, "[1] Amplitude Min/Max"=<0 1>, "[1] Bias"=1)
       )
      )
     )
    )) {isSetParameters=true}
   /SetVariables(Particles.Color_Decal_Update:LinearColor=&MultiplyLinearColor(
     Linear Color=$Particles.Initial.Color_Decal
     Scale Factor=&MakeVector4FromVectorAndFloat(
      "Vector (RGB)"=<1 1 1>
      "Float (Alpha)"=&Multiply_Float(
       A=&Multiply_Float(
        A=&RemapRange(Input Value To Remap=$User.Multiplier, Input Min=0, Input Max=1, Output Min=0.4, Output Max=1)
        B=&Lerp_Float(A=0, B=1, Alpha=$System.Timeline.Time)
       )
       B=&Waveform("Waveform [1]"=::NewEnumerator0, "[1] Frequency"=4.1057, "[1] Amplitude Min/Max"=<0 1>, "[1] Bias"=1)
      )
     )
    )) {enabled=false, isSetParameters=true}
  ]
  rendererClasses [
   &NiagaraSpriteRendererProperties
   &NiagaraLightRendererProperties
   &NiagaraMeshRendererProperties
  ]
  renderers [
   {rendererClass=&NiagaraSpriteRendererProperties}
   {rendererClass=&NiagaraLightRendererProperties, rendererIndex=1}
   {rendererClass=&NiagaraSpriteRendererProperties, rendererIndex=2}
   {rendererClass=&NiagaraMeshRendererProperties, rendererIndex=3}
   {rendererClass=&NiagaraSpriteRendererProperties, rendererIndex=4}
   {rendererClass=&NiagaraSpriteRendererProperties, rendererIndex=5}
  ]
 }
 Core001 {
  bEnabled=false
  emitterUpdateScript [
   &EmitterState(Inactive Response=::NewEnumerator2, Life Cycle Mode=::NewEnumerator1, Loop Duration=4, UseLoopDelay=false)
   &SpawnRate(
    Spawn Probability=&MakeCustomFloatFromBool(
     Bool=&SetBoolByFloatComparison(A=&VectorLength(Vector=$Engine.Owner.Velocity), B=5, Comparison Type=::NewEnumerator6)
     True Float=0
     False Float=1
    )
    SpawnRate=4
    Use Spawn Probability=true
   ) {module=SpawnRate001}
   &SpawnRate(
    Spawn Probability=&MakeCustomFloatFromBool(
     Bool=&SetBoolByFloatComparison(A=&VectorLength(Vector=$Engine.Owner.Velocity), B=5, Comparison Type=::NewEnumerator6)
     True Float=0
     False Float=1
    )
    SpawnGroup=1
    SpawnRate=40
    Use Spawn Probability=true
   )
   &SpawnBurst_Instantaneous(Spawn Count=2, Spawn Group=2, Use Spawn Probability=false, UseLoopCountLimit=false) {module=SpawnBurst_Instantaneous001}
   &SpawnBurst_Instantaneous(Spawn Group=3, Use Spawn Probability=false, UseLoopCountLimit=false)
  ]
  particleSpawnScript [
   /SetVariables(Particles.VisibilityTag=&SpawnGroup()) {isSetParameters=true, module=SetVariables_#7}
   &InitializeParticle(
    Color=<rgba 1.5 0.8757 0.4297 1>
    Lifetime=&MergeHelper_FloatSwitch(
     0=0.36
     1=&UniformRangedFloat(Minimum=0.05, Maximum=0.1)
     2=4
     3=4
     Index=$Particles.VisibilityTag
     Num=::NewEnumerator2
    )
    Mesh Scale Max=&MakeVector(
     X=&RemapRange(Input Value To Remap=$User.Multiplier, Input Min=0, Input Max=1, Output Min=0.4, Output Max=0.5)
     Y=0.2
     Z=0.35
    )
    Mesh Scale Min=&MakeVector(
     X=&RemapRange(Input Value To Remap=$User.Multiplier, Input Min=0, Input Max=1, Output Min=0.3, Output Max=0.35)
     Y=0.2
     Z=0.1
    )
    Mesh Scale Mode=::NewEnumerator2
    Position Mode=::NewEnumerator1
    Position Offset=&MaskVectorBySpawnGroup(Masked Vector=<0 0 0>, Passthrough Vector=<0 0 0>)
    Sprite Size=&MergeHelper_Vector2Switch(0=<130 72>, 1=<0 0>, 2=<15 15>, 3=?, 4=?, Index=$Particles.VisibilityTag, Num=::NewEnumerator1)
    Sprite Size Mode=::NewEnumerator4
    UsePositionOffset=false
    Write Material Random=true
   ) {module=InitializeParticle001}
   &DynamicMaterialParameters(
    Index 0 Param 1=0.2
    Module.Param0WriteEnabled=true
    Module.Param1WriteEnabled=true
    Module.Param2WriteEnabled=true
    Module.Param3WriteEnabled=true
   )
   &CameraOffset(Camera Offset Amount=&MergeHelper_FloatSwitch(0=10, 1=0, 2=5, 3=?, Index=$Particles.VisibilityTag, Num=::NewEnumerator1)) {module=CameraOffset001}
   &SocketLocation(
    "Inherit Velocity (Scale)"=0.01
    Orientation Sampling=::NewEnumerator1
    Position Sampling=::NewEnumerator1
    SamplePosOffset=true
    Sampled Orientation Coordinate Space:ENiagaraCoordinateSpace=::Local
    Sampled Position Offset=<2 0 -2>
    Sampled Velocity Coordinate Space:ENiagaraCoordinateSpace=::Local
    Socket Reader=$System.Socket Reader
    Socket Sampling Mode=::NewEnumerator3
    Velocity Sampling:ENiagara_AttributeSamplingApplyOutput=::NewEnumerator0
    VelocitySpeedLimit=false
   ) {module=SocketLocation001}
   /SetVariables(Particles.PositionGround_Spark:NiagaraPosition=$Output.SocketLocation001.SampledPosition) {isSetParameters=true, module=SetVariables_#8}
   &InitialMeshOrientation(
    Mesh Orientation Mode=::NewEnumerator2
    Orientation Coordinate Space=::Local
    Orientation Vector=<-2 0 -2>
    Rotation=&RandomRangeVector(Minimum=<0.45 -0.1 0>, Maximum=<0.5 -0.1 0>)
    Rotation Coordinate Space=::NewEnumerator0
    Use Rotation Vector=true
   ) {module=InitialMeshOrientation001}
   /SetVariables(Particles.Color_Decal:LinearColor=<rgba 25 8.2694 1.0992 0.3652>, Particles.DynamicMaterialParameter_Decal:Vector4f=<1 0 1 0>) {isSetParameters=true, module=SetVariables_#9}
  ]
  particleUpdateScript [
   &ParticleState()
   &ScaleSpriteSize(
    Scale Sprite Size Mode=::NewEnumerator0
    Uniform Scale Factor=&RemapRange(Input Value To Remap=$User.Multiplier, Input Min=0, Input Max=1, Output Min=0.25, Output Max=1)
   )
   &SolveForcesAndVelocity(Clamp Velocity=false, Limit Acceleration=false)
   &SocketLocation(
    Position Sampling=::NewEnumerator1
    SamplePosOffset=true
    Sampled Position Offset=&MergeHelper_VectorSwitch(0=<0 0 0>, 1=<0 0 0>, 2=<0 0 -5>, 3=?, Index=$Particles.VisibilityTag, Num=::NewEnumerator1)
    Socket Reader=$System.Socket Reader
    Socket Sampling Mode=::NewEnumerator3
    VelocitySpeedLimit=false
   )
   /SetVariables(
    Particles.SparkColor:LinearColor=&MakeLinearColorFromVectorAndFloat(
     "Vector (RGB)"=&MakeVectorFromLinearColorRGB(Color=<rgba 5 2.5329 0.0658 1>)
     "Float (Alpha)"=&Lerp_Float(A=0, B=1, Alpha=$System.Timeline.Time)
    )
    Particles.SparkSize:Vector2f=&MakeVector2D(
     X=500
     Y=&ClampFloat(Float=&Sine(Normalized Angle=$Emitter.Age, Period=0.0616, Scale=4.1323, Bias=4.3498), Min=12.5264, Max=24.152)
    )
    Particles.Vector 2D:Vector2f=&Multiply_Vector2D_ByFloat(
     Float=&RemapRange(Input Value To Remap=$User.Multiplier, Input Min=0, Input Max=1, Output Min=0.6, Output Max=1)
     Vector2D=&Vector2DFromCurve(
      Vector2Curve=DataInterface() {propertyValues {
        XCurve {defaultValue=1, keys [{tangentMode=RCTM_User, time=0, value=0.5}, {tangentMode=RCTM_User, time=1, value=1}]}
        YCurve {defaultValue=1, keys [{tangentMode=RCTM_User, time=0, value=0.5}, {tangentMode=RCTM_User, time=1, value=1}]}
       }}
      ScaleCurve=&RandomRangeVector2D(Minimum=<23.4 8.7388>, Maximum=<23.4 8.3315>)
     )
    )
   ) {isSetParameters=true, module=SetVariables_#10}
   &SubUVAnimation(
    Random Start Frame=true
    "Sprite Renderer "=DataInterface() {propertyValues {SpriteRenderer=&self:Core001_0.NiagaraSpriteRendererProperties_1}}
    SubUV Lookup Index Scale=5.7547
    UseEndFrame=false
    UseStartFrame=false
   )
   /SetVariables(Particles.Position=&ConvertPositionToVector(
     Input Position=&AddVectorToPosition(Position=$Output.SocketLocation.SampledPosition, Vector=<0 0 -0.5>)
     Conversion Operation:ENiagara_LWCConvertPosToVec=::NewEnumerator0
    )) {isSetParameters=true, module=SetVariables_#11}
   &FN_Light_Attributes(
    Color=<rgba 2 1.0278 0 1>
    Create Color Attribute=true
    Create Exponent Attribute=true
    Create Radius Attribute=true
    Exponent=5
    Radius=40
    Scale Color=&Multiply_Float(
     A=1.3
     B=&MakeCustomFloatFromBool(
      Bool=&SetBoolByFloatComparison(A=&VectorLength(Vector=$Engine.Owner.Velocity), B=5)
      True Float=&Multiply_Float(
       A=&RemapRange(Input Value To Remap=$User.Multiplier, Input Min=0, Input Max=1, Output Min=0.6, Output Max=1.5)
       B=$System.FlickerSpeed
      )
      False Float=0
     )
    )
   )
   /SetVariables(Particles.PositionLights:NiagaraPosition=&AddVectorToPosition(Position=$Output.SocketLocation.SampledPosition, Vector=<-10 -2 -5>)) {isSetParameters=true, module=SetVariables_#12}
   /SetVariables(Particles.Color_GroundSpark:LinearColor=&MultiplyLinearColor(
     Linear Color=&LerpLinearColors(
      StartColor=<rgba 25 8.3333 0 1>
      EndColor=<rgba 25 2.0833 0 1>
      LerpFactor=&Multiply_Float(A=$Particles.NormalizedAge, B=1)
     )
     Scale Factor=&MakeVector4_FromFloat(Float=&RemapRange(Input Value To Remap=$User.Multiplier, Input Min=0, Input Max=1, Output Min=0.3, Output Max=0.5))
    )) {isSetParameters=true, module=SetVariables_#13}
   &ScaleMeshSize(Scale Factor=&VectorFromFloat(Value=&FloatFromCurve(
      FloatCurve=DataInterface() {propertyValues {Curve {keys [{arriveTangent=2.2457, interpMode=RCIM_Cubic, leaveTangent=2.2457, tangentMode=RCTM_User, time=0, value=0}, <at 1 1>]}}}
      Scale Curve=&RemapRange(Input Value To Remap=$User.Multiplier, Input Min=0, Input Max=1, Output Min=0.4, Output Max=0.55)
     )))
   /SetVariables(Particles.Color_Decal_Update:LinearColor=&MultiplyLinearColor(
     Linear Color=$Particles.Initial.Color_Decal
     Scale Factor=&MakeVector4FromVectorAndFloat(
      "Vector (RGB)"=<1 1 1>
      "Float (Alpha)"=&Multiply_Float(
       A=&Multiply_Float(
        A=&RemapRange(Input Value To Remap=$User.Multiplier, Input Min=0, Input Max=1, Output Min=0.4, Output Max=1)
        B=&Lerp_Float(A=0, B=1, Alpha=$System.Timeline.Time)
       )
       B=&Waveform("Waveform [1]"=::NewEnumerator0, "[1] Frequency"=4.1057, "[1] Amplitude Min/Max"=<0 1>, "[1] Bias"=1)
      )
     )
    )) {isSetParameters=true}
  ]
  rendererClasses [
   &NiagaraSpriteRendererProperties
   &NiagaraLightRendererProperties
   &NiagaraMeshRendererProperties
  ]
  renderers [
   {rendererClass=&NiagaraSpriteRendererProperties}
   {rendererClass=&NiagaraLightRendererProperties, rendererIndex=1}
   {rendererClass=&NiagaraSpriteRendererProperties, rendererIndex=2}
   {rendererClass=&NiagaraMeshRendererProperties, rendererIndex=3}
   {rendererClass=&NiagaraSpriteRendererProperties, rendererIndex=4}
  ]
 }
 Sparks {
  emitterUpdateScript [
   &EmitterState(Life Cycle Mode=::NewEnumerator1, Loop Delay=2.3833, Loop Duration=2, UseLoopDelay=false)
   &SpawnRate(
    Spawn Probability=&Multiply_Float(
     A=&MakeCustomFloatFromBool(
      Bool=&SetBoolByFloatComparison(A=&VectorLength(Vector=$Engine.Owner.Velocity), B=5, Comparison Type=::NewEnumerator6)
      True Float=0
      False Float=&MakeCustomFloatFromBool(
       Bool=&SetBoolByFloatComparison(A=&VectorLength(Vector=$Engine.Owner.Velocity), B=500, Comparison Type=::NewEnumerator6)
       True Float=1
       False Float=0
      )
     )
     B=&RemapRange(Input Value To Remap=$User.Multiplier, Input Min=0, Input Max=1, Output Min=0.25, Output Max=1)
    )
    SpawnRate=40
    Use Spawn Probability=true
   )
  ]
  particleSpawnScript [
   /SetVariables(Particles.ParticleRandom:NiagaraFloat=&SelectFloatFromArray(
     Direct Array Index=&SelectIntFromWeightedDistributionArray(
      Array Sampling Mode:ENiagara_LifetimeMode=::NewEnumerator1
      Direct Array Index:NiagaraInt32=?
      Direct Array Mode:ENiagara_WrapClamp=::NewEnumerator0
      Evaluation Type:ENiagaraRandomnessEvaluation=::NewEnumerator0
      Recalculate Random Each Loop:NiagaraBool=false
      Weighted Distribution Int Array:NiagaraDataInterfaceArrayDistributionInt=DataInterface() {propertyValues {ArrayData [{value=0, weight=0.75}, {value=1, weight=0.25}]}}
      Randomness Mode:ENiagaraRandomnessMode=::NewEnumerator0
      Random Seed:NiagaraInt32=0
      Fixed Random Seed:NiagaraBool=false
     )
     Float Selection Array=DataInterface() {propertyValues {FloatData [0,1]}}
    )) {isSetParameters=true, module=SetVariables_#4}
   &InitializeParticle(
    Color Mode=::NewEnumerator0
    Lifetime=&MergeHelper_FloatSwitch(
     0=&RandomRangeFloat(Minimum=0.18, Maximum=0.25)
     1=&RandomRangeFloat(Minimum=0.4, Maximum=0.6)
     2=?
     3=?
     Index=&MakeIntFromFloat(Float=$Particles.ParticleRandom, Conversion Mode=::NewEnumerator1)
     Num=::NewEnumerator0
    )
    Mesh Scale Max=<0.0663 0.0663 0.0663>
    Mesh Scale Min=<3 3 3>
    Mesh Scale Mode=::NewEnumerator2
    Position Mode=::NewEnumerator0
    Sprite Rotation Mode=::NewEnumerator2
    Sprite Size Max=<3 3>
    Sprite Size Min=<1 1>
    Sprite Size Mode=::NewEnumerator2
    Write Material Random=true
   )
   &SocketLocation(
    "Inherit Velocity (Scale)"=-2
    SamplePosOffset=false
    Sampled Orientation Coordinate Space:ENiagaraCoordinateSpace=::Local
    Sampled Position Offset=<0 0 -1>
    Sampled Velocity Coordinate Space:ENiagaraCoordinateSpace=::Simulation
    Socket Reader=$System.Socket Reader
    Socket Sampling Mode=::NewEnumerator3
    VelocitySpeedLimit=false
   )
   &AddVelocity(
    Cone Angle=&MergeHelper_FloatSwitch(
     0=40
     1=50
     2=?
     3=?
     Index=&MakeIntFromFloat(Float=$Particles.ParticleRandom, Conversion Mode=::NewEnumerator1)
     Num=::NewEnumerator0
    )
    Cone Axis=<0.3 1 -0.25>
    Module.UseConeFalloff=false
    Module.UseVelDistribution=false
    Override Seed=false
    Velocity Mode=::NewEnumerator2
    Velocity Speed=&Multiply_Float(
     A=&MergeHelper_FloatSwitch(
      0=&RandomRangeFloat(Minimum=50, Maximum=100)
      1=&RandomRangeFloat(Minimum=75, Maximum=150)
      2=?
      3=?
      Index=&MakeIntFromFloat(Float=$Particles.ParticleRandom, Conversion Mode=::NewEnumerator1)
      Num=::NewEnumerator0
     )
     B=&RemapRange(Input Value To Remap=$User.Multiplier, Input Min=0, Input Max=1, Output Min=1.7, Output Max=2.2)
    )
   ) {module=AddVelocity002}
   &AddVelocity(
    Cone Angle=&MergeHelper_FloatSwitch(
     0=40
     1=30
     2=?
     3=?
     Index=&MakeIntFromFloat(Float=$Particles.ParticleRandom, Conversion Mode=::NewEnumerator1)
     Num=::NewEnumerator0
    )
    Cone Axis=&RotateVectorByQuaternion(Vector To Rotate=<-0.5 0.45 0.15>, Quaternion=$Output.SocketLocation.SampledRotation)
    Module.UseConeFalloff=false
    Module.UseVelDistribution=false
    Override Seed=false
    Velocity Mode=::NewEnumerator2
    Velocity Speed=&Multiply_Float(
     A=&MergeHelper_FloatSwitch(
      0=&RandomRangeFloat(Minimum=50, Maximum=150)
      1=&RandomRangeFloat(Minimum=75, Maximum=200)
      2=?
      3=?
      Index=&MakeIntFromFloat(Float=$Particles.ParticleRandom, Conversion Mode=::NewEnumerator1)
      Num=::NewEnumerator0
     )
     B=&RemapRange(Input Value To Remap=$User.Multiplier, Input Min=0, Input Max=1, Output Min=0.65, Output Max=1.35)
    )
   ) {enabled=false}
   &InheritVelocity(Inherited Velocity Amount Scale=<500 500 500>, Inherited Velocity Speed Limit=1500, Speed Limit=true, UseSpeedThreshold=false) {enabled=false}
  ]
  particleUpdateScript [
   &ParticleState()
   /SetVariables(
    Particles.Flicker:NiagaraFloat=&Waveform("Waveform [1]"=::NewEnumerator0, "[1] Frequency"=75, "[1] Amplitude Min/Max"=<0 2>)
    Particles.Flicker Size:NiagaraFloat=&Waveform("Waveform [1]"=::NewEnumerator0, "[1] Frequency"=75, "[1] Amplitude Min/Max"=<0.5 1>)
    Particles.Flicker Value:NiagaraFloat=&Waveform("Waveform [1]"=::NewEnumerator0, "[1] Frequency"=75, "[1] Amplitude Min/Max"=<2 8>)
   ) {isSetParameters=true, module=SetVariables_#5}
   &ScaleSpriteSize(
    Scale Sprite Size Mode=::NewEnumerator0
    Uniform Scale Factor=&Lerp_Float(
     A=1
     B=&Multiply_Float(A=$Particles.Flicker Size, B=1.5)
     Alpha=&FloatFromCurve(FloatCurve=DataInterface() {propertyValues {Curve {keys [<at 0 0.1076>,<at 0.2485 0.6128>,<at 0.45 1>,<at 1 1>]}}})
    )
   )
   &ScaleColor(
    Scale RGB=&MakeVectorFromLinearColorRGB(Color=&ColorFromCurve(ColorCurve=DataInterface() {propertyValues {
        AlphaCurve {keys [<at 0 1>]}
        BlueCurve {keys [<at 0 0.2604>,<at 0.1639 0>,<at 1 0>]}
        GreenCurve {keys [<at 0 1.0503>,<at 0.1639 3.75>,<at 1 0.25>]}
        RedCurve {keys [<at 0 5>,<at 0.1639 15>,<at 1 1>]}
       }}))
    ScaleA=true
    ScaleRGB=true
   )
   &GravityForce(Gravity=&MergeHelper_VectorSwitch(
     0=&RandomRangeVector(Minimum=<0 0 -500>, Maximum=<0 0 -500>)
     1=<0 0 -1200>
     2=?
     3=?
     Index=&MakeIntFromFloat(Float=$Particles.ParticleRandom, Conversion Mode=::NewEnumerator1)
     Num=::NewEnumerator0
    ))
   &Collision(
    Advanced Aging Rate=1
    Age Colliding Particles=false
    Collision Enabled=&SetBoolByFloatComparison(A=$Particles.NormalizedAge, B=0.2)
    Enable Rest State=false
    Friction=0.6
    Friction During a Bounce=0.4
    Kill On Collision:NiagaraInt32=&RandomRangeInt(Minimum=3, Maximum=5)
    KillOnCollision=false
    Max Number Of Collisions=5
    Module.EnableMaxCollisionCount=false
    Module.Randomize Collision Normal Vector=false
    Restitution=0
   ) {module=Collision001}
   &Drag(Drag=&RandomRangeFloat(Minimum=4.5, Maximum=7), Use Linear Drag=true, Use Rotational Drag=false)
   &SolveForcesAndVelocity(Clamp Velocity=false, Limit Acceleration=false)
   &ScaleSpriteSizeBySpeed(Max Scale Factor=<0.5 3>, Min Scale Factor=<0.5 0.5>, Velocity Threshold=125) {module=ScaleSpriteSizeBySpeed001}
   /SetVariables(Particles.Color=&MultiplyLinearColor(
     Linear Color=&LerpLinearColors(
      StartColor=<rgba 25 8.3333 0 1>
      EndColor=<rgba 25 2.5 0 1>
      LerpFactor=&Multiply_Float(A=$Particles.NormalizedAge, B=1.5)
     )
     Scale Factor=&MakeVector4_FromFloat(Float=1)
    )) {isSetParameters=true, module=SetVariables_#6}
   &BlendSpace(
    "Position Blend (World <-> Local)":NiagaraFloat=&MakeCustomFloatFromBool(Bool=$Particles.HasCollided, True Float=0, False Float=0.1, Translation Contribution:NiagaraFloat=1, Rotation Contribution:NiagaraFloat=1)
    Position Coordinate Space=1
   ) {module=BlendSpace001}
   /SetVariables(Particles.SpriteSizetest:Vector2f=<5 5>) {isSetParameters=true, module=SetVariables_#14}
  ]
  rendererClasses [&NiagaraSpriteRendererProperties, &NiagaraMeshRendererProperties]
  renderers [
   {rendererClass=&NiagaraSpriteRendererProperties}
   {rendererClass=&NiagaraSpriteRendererProperties, rendererIndex=1}
   {rendererClass=&NiagaraMeshRendererProperties, rendererIndex=2}
  ]
 }
 Sparks_Rails {
  emitterUpdateScript [
   &EmitterState(Life Cycle Mode=::NewEnumerator1, Loop Delay=2.3833, Loop Duration=2, UseLoopDelay=false)
   &SpawnRate(
    Spawn Probability=&Multiply_Float(
     A=&MakeCustomFloatFromBool(
      Bool=&SetBoolByFloatComparison(A=&VectorLength(Vector=$Engine.Owner.Velocity), B=500)
      True Float=1
      False Float=0
     )
     B=&RemapRange(Input Value To Remap=$User.Multiplier, Input Min=0, Input Max=1, Output Min=0.25, Output Max=1)
    )
    SpawnRate=40
    Use Spawn Probability=true
   )
  ]
  particleSpawnScript [
   /SetVariables(Particles.ParticleRandom:NiagaraFloat=&SelectFloatFromArray(
     Direct Array Index=&SelectIntFromWeightedDistributionArray(
      Array Sampling Mode:ENiagara_LifetimeMode=::NewEnumerator1
      Direct Array Index:NiagaraInt32=?
      Direct Array Mode:ENiagara_WrapClamp=::NewEnumerator0
      Evaluation Type:ENiagaraRandomnessEvaluation=::NewEnumerator0
      Recalculate Random Each Loop:NiagaraBool=false
      Weighted Distribution Int Array:NiagaraDataInterfaceArrayDistributionInt=DataInterface() {propertyValues {ArrayData [{value=0, weight=0.75}, {value=1, weight=0.25}]}}
      Randomness Mode:ENiagaraRandomnessMode=::NewEnumerator0
      Random Seed:NiagaraInt32=0
      Fixed Random Seed:NiagaraBool=false
     )
     Float Selection Array=DataInterface() {propertyValues {FloatData [0,1]}}
    )) {isSetParameters=true, module=SetVariables_#4}
   &InitializeParticle(
    Color Mode=::NewEnumerator0
    Lifetime=&MergeHelper_FloatSwitch(
     0=&RandomRangeFloat(Minimum=0.18, Maximum=0.25)
     1=&RandomRangeFloat(Minimum=0.4, Maximum=0.6)
     2=?
     3=?
     Index=&MakeIntFromFloat(Float=$Particles.ParticleRandom, Conversion Mode=::NewEnumerator1)
     Num=::NewEnumerator0
    )
    Mesh Scale Max=<0.0663 0.0663 0.0663>
    Mesh Scale Min=<3 3 3>
    Mesh Scale Mode=::NewEnumerator2
    Position Mode=::NewEnumerator0
    Sprite Rotation Mode=::NewEnumerator2
    Sprite Size Max=<3 3>
    Sprite Size Min=<1 1>
    Sprite Size Mode=::NewEnumerator2
    Write Material Random=true
   )
   &SocketLocation(
    "Inherit Velocity (Scale)"=0.2
    SamplePosOffset=false
    Sampled Orientation Coordinate Space:ENiagaraCoordinateSpace=::Local
    Sampled Position Offset=<0 0 -1>
    Sampled Velocity Coordinate Space:ENiagaraCoordinateSpace=::Simulation
    Socket Reader=$System.Socket Reader
    Socket Sampling Mode=::NewEnumerator3
    VelocitySpeedLimit=false
   )
   &AddVelocity(
    Cone Angle=&MergeHelper_FloatSwitch(
     0=40
     1=50
     2=?
     3=?
     Index=&MakeIntFromFloat(Float=$Particles.ParticleRandom, Conversion Mode=::NewEnumerator1)
     Num=::NewEnumerator0
    )
    Cone Axis=<0.3 1 -0.25>
    Module.UseConeFalloff=false
    Module.UseVelDistribution=false
    Override Seed=false
    Velocity Mode=::NewEnumerator2
    Velocity Speed=&Multiply_Float(
     A=&MergeHelper_FloatSwitch(
      0=&RandomRangeFloat(Minimum=50, Maximum=100)
      1=&RandomRangeFloat(Minimum=75, Maximum=150)
      2=?
      3=?
      Index=&MakeIntFromFloat(Float=$Particles.ParticleRandom, Conversion Mode=::NewEnumerator1)
      Num=::NewEnumerator0
     )
     B=&RemapRange(Input Value To Remap=$User.Multiplier, Input Min=0, Input Max=1, Output Min=1.7, Output Max=2.2)
    )
   ) {module=AddVelocity002}
   &AddVelocity(
    Cone Angle=&MergeHelper_FloatSwitch(
     0=40
     1=30
     2=?
     3=?
     Index=&MakeIntFromFloat(Float=$Particles.ParticleRandom, Conversion Mode=::NewEnumerator1)
     Num=::NewEnumerator0
    )
    Cone Axis=&RotateVectorByQuaternion(Vector To Rotate=<-0.5 0.45 0.15>, Quaternion=$Output.SocketLocation.SampledRotation)
    Module.UseConeFalloff=false
    Module.UseVelDistribution=false
    Override Seed=false
    Velocity Mode=::NewEnumerator2
    Velocity Speed=&Multiply_Float(
     A=&MergeHelper_FloatSwitch(
      0=&RandomRangeFloat(Minimum=50, Maximum=150)
      1=&RandomRangeFloat(Minimum=75, Maximum=200)
      2=?
      3=?
      Index=&MakeIntFromFloat(Float=$Particles.ParticleRandom, Conversion Mode=::NewEnumerator1)
      Num=::NewEnumerator0
     )
     B=&RemapRange(Input Value To Remap=$User.Multiplier, Input Min=0, Input Max=1, Output Min=0.65, Output Max=1.35)
    )
   ) {enabled=false}
   &InheritVelocity(Inherited Velocity Amount Scale=<500 500 500>, Inherited Velocity Speed Limit=1500, Speed Limit=true, UseSpeedThreshold=false) {enabled=false}
  ]
  particleUpdateScript [
   &ParticleState()
   /SetVariables(
    Particles.Flicker:NiagaraFloat=&Waveform("Waveform [1]"=::NewEnumerator0, "[1] Frequency"=75, "[1] Amplitude Min/Max"=<0 2>)
    Particles.Flicker Size:NiagaraFloat=&Waveform("Waveform [1]"=::NewEnumerator0, "[1] Frequency"=75, "[1] Amplitude Min/Max"=<0.5 1>)
    Particles.Flicker Value:NiagaraFloat=&Waveform("Waveform [1]"=::NewEnumerator0, "[1] Frequency"=75, "[1] Amplitude Min/Max"=<2 8>)
   ) {isSetParameters=true, module=SetVariables_#5}
   &ScaleSpriteSize(
    Scale Sprite Size Mode=::NewEnumerator0
    Uniform Scale Factor=&Lerp_Float(
     A=1
     B=&Multiply_Float(A=$Particles.Flicker Size, B=1.5)
     Alpha=&FloatFromCurve(FloatCurve=DataInterface() {propertyValues {Curve {keys [<at 0 0.1076>,<at 0.2485 0.6128>,<at 0.45 1>,<at 1 1>]}}})
    )
   )
   &ScaleColor(
    Scale RGB=&MakeVectorFromLinearColorRGB(Color=&ColorFromCurve(ColorCurve=DataInterface() {propertyValues {
        AlphaCurve {keys [<at 0 1>]}
        BlueCurve {keys [<at 0 0.2604>,<at 0.1639 0>,<at 1 0>]}
        GreenCurve {keys [<at 0 1.0503>,<at 0.1639 3.75>,<at 1 0.25>]}
        RedCurve {keys [<at 0 5>,<at 0.1639 15>,<at 1 1>]}
       }}))
    ScaleA=true
    ScaleRGB=true
   )
   &GravityForce(Gravity=&MergeHelper_VectorSwitch(
     0=&RandomRangeVector(Minimum=<0 0 -500>, Maximum=<0 0 -500>)
     1=<0 0 -1200>
     2=?
     3=?
     Index=&MakeIntFromFloat(Float=$Particles.ParticleRandom, Conversion Mode=::NewEnumerator1)
     Num=::NewEnumerator0
    ))
   &Collision(
    Advanced Aging Rate=1
    Age Colliding Particles=false
    Collision Enabled=&SetBoolByFloatComparison(A=$Particles.NormalizedAge, B=0.2)
    Enable Rest State=false
    Friction=0.6
    Friction During a Bounce=0.4
    Kill On Collision:NiagaraInt32=&RandomRangeInt(Minimum=3, Maximum=5)
    KillOnCollision=false
    Max Number Of Collisions=5
    Module.EnableMaxCollisionCount=false
    Module.Randomize Collision Normal Vector=false
    Restitution=0
   ) {module=Collision001}
   &Drag(Drag=&RandomRangeFloat(Minimum=4.5, Maximum=7), Use Linear Drag=true, Use Rotational Drag=false)
   &SolveForcesAndVelocity(Clamp Velocity=false, Limit Acceleration=false)
   &ScaleSpriteSizeBySpeed(Max Scale Factor=<0.5 3>, Min Scale Factor=<0.5 0.5>, Velocity Threshold=125) {module=ScaleSpriteSizeBySpeed001}
   /SetVariables(Particles.Color=&MultiplyLinearColor(
     Linear Color=&LerpLinearColors(
      StartColor=<rgba 25 8.3333 0 1>
      EndColor=<rgba 25 2.5 0 1>
      LerpFactor=&Multiply_Float(A=$Particles.NormalizedAge, B=1.5)
     )
     Scale Factor=&MakeVector4_FromFloat(Float=1)
    )) {isSetParameters=true, module=SetVariables_#6}
   &BlendSpace(
    Position Coordinate Space=2
    Rotation Contribution=0
    Translation Contribution:NiagaraFloat=&FloatFromCurve(FloatCurve=DataInterface() {propertyValues {Curve {keys [<at 0.1683 0>,<at 0.2 1>]}}}, Scale Curve=0.9)
   )
   /SetVariables(Particles.SpriteSizetest:Vector2f=<5 5>) {isSetParameters=true, module=SetVariables_#14}
  ]
  rendererClasses [&NiagaraSpriteRendererProperties, &NiagaraMeshRendererProperties]
  renderers [
   {rendererClass=&NiagaraSpriteRendererProperties}
   {rendererClass=&NiagaraSpriteRendererProperties, rendererIndex=1}
   {rendererClass=&NiagaraMeshRendererProperties, rendererIndex=2}
  ]
 }
 Trail {
  emitterUpdateScript [
   &EmitterState(Life Cycle Mode=::NewEnumerator1, Loop Duration=1, UseLoopDelay=false)
   &SpawnPerFrame(Use Spawn Probability=false) {enabled=false, module=SpawnPerFrame001}
   &SpawnPerFrame(
    Spawn Probability:NiagaraFloat=&MakeCustomFloatFromBool(
     Bool=&SetBoolByFloatComparison(A=&VectorLength(Vector=$Engine.Owner.Velocity), B=5, Comparison Type=::NewEnumerator6)
     True Float=0
     False Float=&MakeCustomFloatFromBool(
      Bool=&SetBoolByFloatComparison(A=&VectorLength(Vector=$Engine.Owner.Velocity), B=175)
      True Float=0
      False Float=1
     )
    )
    Use Spawn Probability=true
   )
   &SpawnRate(
    Spawn Probability=&MakeCustomFloatFromBool(
     Bool=&SetBoolByFloatComparison(A=&VectorLength(Vector=$Engine.Owner.Velocity), B=5, Comparison Type=::NewEnumerator6)
     True Float=0
     False Float=&MakeCustomFloatFromBool(
      Bool=&SetBoolByFloatComparison(A=&VectorLength(Vector=$Engine.Owner.Velocity), B=200)
      True Float=0
      False Float=1
     )
    )
    SpawnRate=200
    Use Spawn Probability=true
   ) {enabled=false}
  ]
  particleSpawnScript [
   &InitializeParticle(
    Color=<rgba 0.9389 1.1267 0 0.8>
    Lifetime=0.12
    Ribbon Width=20
    Ribbon Width Mode=::NewEnumerator1
    UsePositionOffset=false
    Write Material Random=true
   )
   &SocketLocation(
    SamplePosOffset=true
    Sampled Position Offset=<0 0 -2>
    Socket Reader=DataInterface() {propertyValues {
      AttachComponentClass=null
      AttachComponentTag=None
      EditorPreviewAsset=/BRCosmetics/Weapons/FORT_Melee/Pickaxe_NobleCrusher/Meshes/NobleCrusher_Axe
      SourceActorRef=null
      SourceAsset=null
      SourceMode=Default
      bRequireCurrentFrameData=true
      bUpdateSocketsPerFrame=true
      FilteredSockets [Emote_Socket]
      ObjectParameterBinding {parameter {name=None, typeDefHandle {registeredTypeIndex=1}, varData []}}
     }}
    Socket Sampling Mode=::NewEnumerator3
    VelocitySpeedLimit=false
   )
   /SetVariables(Particles.RibbonFacing=<0 0 0.15>) {isSetParameters=true, module=SetVariables_#15}
   &AddVelocity(
    Cone Angle=0
    Cone Axis=<0.3 0.7 0.6>
    Module.UseConeFalloff=false
    Module.UseVelDistribution=false
    Override Seed=false
    Velocity Mode=::NewEnumerator2
    Velocity Speed=100
   ) {module=AddVelocity003}
  ]
  particleUpdateScript [
   &ParticleState()
   &SolveForcesAndVelocity(Clamp Velocity=false, Limit Acceleration=false)
   &DynamicMaterialParameters(
    Index 0 Param 4=&FloatFromCurve(FloatCurve=DataInterface() {propertyValues {Curve {
        defaultValue=1
        keys [
         {arriveTangent=-1.505, interpMode=RCIM_Cubic, leaveTangent=-1.505, tangentMode=RCTM_User, time=0, value=0.75}
         {arriveTangent=0.2345, interpMode=RCIM_Cubic, leaveTangent=0.2345, tangentMode=RCTM_User, time=0.9, value=0}
        ]
       }}})
    Module.Index0Float0Write=false
    Module.Index0Float1Write=false
    Module.Index0Float2Write=false
    Module.Index0Float3Write=true
   )
   &ScaleColor(
    Scale Alpha=&FloatFromCurve(
     FloatCurve=DataInterface() {propertyValues {Curve {keys [
         {interpMode=RCIM_Cubic, leaveTangent=-3.4022, tangentMode=RCTM_Break, time=0.0546, value=1}
         {arriveTangent=-0.4244, tangentMode=RCTM_User, time=1, value=0}
        ]}}}
     Scale Curve=2
    )
    Scale RGB=&VectorFromFloat(Value=&FloatFromCurve(
      FloatCurve=DataInterface() {propertyValues {Curve {keys [{arriveTangent=-2.5017, interpMode=RCIM_Cubic, leaveTangent=-2.5017, tangentMode=RCTM_User, time=0, value=1}, <at 0.3 0.1 RCIM_Cubic>]}}}
      Scale Curve=50
     ))
    ScaleA=true
    ScaleRGB=true
   )
   &ScaleRibbonWidth(Ribbon Width Scale=&FloatFromCurve(FloatCurve=DataInterface() {propertyValues {Curve {
        defaultValue=0.6743
        keys [
         {arriveTangent=-0.2411, interpMode=RCIM_Cubic, leaveTangent=-0.2411, tangentMode=RCTM_User, time=0, value=1}
         {arriveTangent=-3.5395, interpMode=RCIM_Cubic, leaveTangent=-3.5395, tangentMode=RCTM_User, time=1, value=0}
        ]
       }}}))
  ]
  rendererClasses [&NiagaraRibbonRendererProperties]
  renderers [{rendererClass=&NiagaraRibbonRendererProperties}]
 }
 Trail001 {
  bEnabled=false
  emitterUpdateScript [
   &EmitterState(Life Cycle Mode=::NewEnumerator1, Loop Duration=1, UseLoopDelay=false)
   &SpawnPerFrame(
    Spawn Probability:NiagaraFloat=&MakeCustomFloatFromBool(
     Bool=&SetBoolByFloatComparison(A=&VectorLength(Vector=$Engine.Owner.Velocity), B=5, Comparison Type=::NewEnumerator6)
     True Float=0
     False Float=&MakeCustomFloatFromBool(
      Bool=&SetBoolByFloatComparison(A=&VectorLength(Vector=$Engine.Owner.Velocity), B=200)
      True Float=0
      False Float=1
     )
    )
    Use Spawn Probability=true
   )
   &SpawnRate(
    Spawn Probability=&MakeCustomFloatFromBool(
     Bool=&SetBoolByFloatComparison(A=&VectorLength(Vector=$Engine.Owner.Velocity), B=5, Comparison Type=::NewEnumerator6)
     True Float=0
     False Float=&MakeCustomFloatFromBool(
      Bool=&SetBoolByFloatComparison(A=&VectorLength(Vector=$Engine.Owner.Velocity), B=200)
      True Float=0
      False Float=1
     )
    )
    SpawnRate=200
    Use Spawn Probability=true
   ) {enabled=false}
  ]
  particleSpawnScript [
   &InitializeParticle(
    Color=<rgba 0.9389 1.1267 0 0.8>
    Lifetime=0.12
    Ribbon Width=20
    Ribbon Width Mode=::NewEnumerator1
    UsePositionOffset=false
    Write Material Random=true
   )
   &SocketLocation(
    SamplePosOffset=false
    Socket Reader=DataInterface() {propertyValues {
      AttachComponentClass=null
      AttachComponentTag=None
      EditorPreviewAsset=/BRCosmetics/Animation/Props/Gadgets_NobleCrusher/NobleCrusher_Gadget
      SourceActorRef=null
      SourceAsset=null
      SourceMode=Default
      bRequireCurrentFrameData=true
      bUpdateSocketsPerFrame=true
      FilteredSockets [Emote_Socket]
      ObjectParameterBinding {parameter {name=None, typeDefHandle {registeredTypeIndex=1}, varData []}}
     }}
    Socket Sampling Mode=::NewEnumerator3
    VelocitySpeedLimit=false
   )
   /SetVariables(Particles.RibbonFacing=<0 0 0.15>) {isSetParameters=true, module=SetVariables_#15}
   &AddVelocity(
    Cone Angle=0
    Cone Axis=<-0.7 0.3 0.6>
    Module.UseConeFalloff=false
    Module.UseVelDistribution=false
    Override Seed=false
    Velocity Mode=::NewEnumerator2
    Velocity Speed=100
   ) {module=AddVelocity003}
  ]
  particleUpdateScript [
   &ParticleState()
   &SolveForcesAndVelocity(Clamp Velocity=false, Limit Acceleration=false)
   &DynamicMaterialParameters(
    Index 0 Param 4=&FloatFromCurve(FloatCurve=DataInterface() {propertyValues {Curve {
        defaultValue=1
        keys [
         {arriveTangent=-1.505, interpMode=RCIM_Cubic, leaveTangent=-1.505, tangentMode=RCTM_User, time=0, value=0.75}
         {arriveTangent=0.2345, interpMode=RCIM_Cubic, leaveTangent=0.2345, tangentMode=RCTM_User, time=0.9, value=0}
        ]
       }}})
    Module.Index0Float0Write=false
    Module.Index0Float1Write=false
    Module.Index0Float2Write=false
    Module.Index0Float3Write=true
   )
   &ScaleColor(
    Scale Alpha=&FloatFromCurve(
     FloatCurve=DataInterface() {propertyValues {Curve {keys [
         {interpMode=RCIM_Cubic, leaveTangent=-3.4022, tangentMode=RCTM_Break, time=0.0546, value=1}
         {arriveTangent=-0.4244, tangentMode=RCTM_User, time=1, value=0}
        ]}}}
     Scale Curve=2
    )
    Scale RGB=&VectorFromFloat(Value=&FloatFromCurve(
      FloatCurve=DataInterface() {propertyValues {Curve {keys [{arriveTangent=-2.5017, interpMode=RCIM_Cubic, leaveTangent=-2.5017, tangentMode=RCTM_User, time=0, value=1}, <at 0.3 0.1 RCIM_Cubic>]}}}
      Scale Curve=50
     ))
    ScaleA=true
    ScaleRGB=true
   )
   &ScaleRibbonWidth(Ribbon Width Scale=&FloatFromCurve(FloatCurve=DataInterface() {propertyValues {Curve {
        defaultValue=0.6743
        keys [
         {arriveTangent=-0.2411, interpMode=RCIM_Cubic, leaveTangent=-0.2411, tangentMode=RCTM_User, time=0, value=1}
         {arriveTangent=-3.5395, interpMode=RCIM_Cubic, leaveTangent=-3.5395, tangentMode=RCTM_User, time=1, value=0}
        ]
       }}}))
  ]
  rendererClasses [&NiagaraRibbonRendererProperties]
  renderers [{rendererClass=&NiagaraRibbonRendererProperties}]
 }
}
summary {
 systemName=NS_Emote_NobleCrusher_GroundSparks
 emitters [
  {emitterName=Sparks, rendererClasses [&NiagaraSpriteRendererProperties, &NiagaraMeshRendererProperties]}
  {emitterName=Trail, rendererClasses [&NiagaraRibbonRendererProperties]}
  {
   emitterName=Core
   rendererClasses [
    &NiagaraSpriteRendererProperties
    &NiagaraLightRendererProperties
    &NiagaraMeshRendererProperties
   ]
  }
  {bEnabled=false, emitterName=Classic_Sparks, rendererClasses [&NiagaraSpriteRendererProperties]}
  {bEnabled=false, emitterName=Trail001, rendererClasses [&NiagaraRibbonRendererProperties]}
  {
   bEnabled=false
   emitterName=Core001
   rendererClasses [
    &NiagaraSpriteRendererProperties
    &NiagaraLightRendererProperties
    &NiagaraMeshRendererProperties
   ]
  }
  {emitterName=Sparks_Rails, rendererClasses [&NiagaraSpriteRendererProperties, &NiagaraMeshRendererProperties]}
 ]
 userVariables [
  User.bIsFrontEnd {type=NiagaraBool}
  User.bIsFrontEndPreview {type=NiagaraBool}
  User.ForwardDot {type=NiagaraFloat, defaultValue=1, description=""}
  User.FrontEndScale {type=NiagaraFloat, defaultValue=0.6, description=""}
  User.LerpCurve {type=NiagaraFloat, defaultValue=0, description=""}
  User.Multiplier {type=NiagaraFloat, defaultValue=1, description=""}
  User.RibbonIDOffset {type=NiagaraInt32}
  User.SprintColor_Hue {type=NiagaraFloat, defaultValue=0, description=""}
  User.SprintColor_Saturation {type=NiagaraFloat, defaultValue=1, description=""}
  User.SwingSectionIndex {type=NiagaraInt32}
  User.VariantID {type=NiagaraFloat, defaultValue=1, description=""}
 ]
}
