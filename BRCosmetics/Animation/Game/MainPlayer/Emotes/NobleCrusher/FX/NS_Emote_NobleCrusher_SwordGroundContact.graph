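# One-shot Niagara effect for the Noble Crusher emote's sword hitting the ground, throwing a warm flash and sparks.
# Flash001 bursts three spawn groups, tagged by VisibilityTag, for mesh, sprite and light renderers, with distance-scaled size and light.
# Sparks randomly splits particles into two lifetime and gravity classes, with flicker, speed stretching and collision for one class, gated by platform quality.
# Flash, Light001 and Sparks001 are disabled variants of these emitters.
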NiagaraSystem /BRCosmetics/Animation/Game/MainPlayer/Emotes/NobleCrusher/FX/NS_Emote_NobleCrusher_SwordGroundContact
dependencies {
 usedDataInterfaces [
  /Script/Niagara.NiagaraDataInterfaceArrayDistributionInt
  &NiagaraDataInterfaceArrayFloat
  &NiagaraDataInterfaceColorCurve
  &NiagaraDataInterfaceCurve
  &NiagaraDataInterfaceVectorCurve
  &NiagaraDataInterfacePlatformSet
 ]
 usedDynamicInputs [
  &RandomRangeVector2D
  &MakeFloatFromVector2d
  &MakeVector4_FromLinearColor
  /Niagara/DynamicInputs/Helpers/SimulationPosition
 ]
 usedModules [&SystemState]
}
emitters {
 Flash {
  bEnabled=false
  emitterUpdateScript [
   &EmitterState(
    Enable Distance Culling=true
    Life Cycle Mode=::NewEnumerator1
    Loop Behavior=::NewEnumerator1
    Loop Duration=1
    Max Distance Response=::KillAfterParticlesFinish
    MaxDistance=30000
    Module.bUseMaxDistance=true
    Module.bUseMinDistance=false
    Reset Age On Awaken=true
    Scalability Mode=::NewEnumerator1
    UseLoopDelay=false
   )
   &SpawnBurst_Instantaneous(Spawn Count=&RandomRangeInt(Minimum=3, Maximum=4), Use Spawn Probability=false)
   &SpawnBurst_Instantaneous(Spawn Group=1, Use Spawn Probability=false, UseLoopCountLimit=false) {module=SpawnBurst_Instantaneous001}
  ]
  particleSpawnScript [
   /SetVariables(Particles.VisibilityTag=&SpawnGroup()) {isSetParameters=true, module=SetVariables_#1}
   &InitializeParticle(
    Lifetime=&MergeHelper_FloatSwitch(
     0=&RandomRangeFloat(Minimum=0.1, Maximum=0.125)
     1=0.2
     2=?
     3=?
     Index=$Particles.VisibilityTag
     Num=::NewEnumerator0
    )
    Mesh Scale=&UniformRangedVector(Minimum=<0.35 0.1 0.1>, Maximum=<0.5 0.2 0.2>)
    Mesh Scale Mode=::NewEnumerator4
    Sprite Size Mode=::NewEnumerator3
    Uniform Sprite Size=80
    UsePositionOffset=false
    Write Material Random=true
   ) {module=InitializeParticle001}
   &CameraOffset(Camera Offset Amount=10)
   /SetVariables(Particles.LightPositionNew:Vector3f=&Add_Vector(
     A=Linked() {linkedVariable {name=Particles.Position, type=NiagaraPosition}}
     B=/Niagara/DynamicInputs/Transforms/TransformPosition(Position=<25 0 0>, Source Space=::Local, Destination Space=::Simulation)
    )) {enabled=false, isSetParameters=true, module=SetVariables_#2}
   &InitialMeshOrientation(
    Mesh Orientation Mode=::NewEnumerator2
    Orientation Coordinate Space=::Local
    Orientation Vector=<-2 0 -2>
    Rotation=&RandomRangeVector(Minimum=<-0.25 -0.25 0>, Maximum=<0.25 0.25 0>)
    Rotation Coordinate Space=::NewEnumerator0
    Use Rotation Vector=true
   ) {module=InitialMeshOrientation001}
  ]
  particleUpdateScript [
   /SetVariables(Particles.DistanceBlend:NiagaraFloat=/Game/Effects/Niagara/DynamicInputs/Math/FN_Fit(
     Clamp:NiagaraBool=true
     Value:NiagaraFloat=$Emitter.Scalability.DistanceFraction
     SourceMin:NiagaraFloat=0.1
     SourceMax:NiagaraFloat=1
     DestMin:NiagaraFloat=0
     DestMax:NiagaraFloat=1
    )) {isSetParameters=true, module=SetVariables_#3}
   &ParticleState()
   &ScaleColor(
    Scale Alpha=&MergeHelper_FloatSwitch(
     0=&FloatFromCurve(
      FloatCurve=DataInterface() {propertyValues {Curve {keys [<at 0 1 RCIM_Cubic>, {arriveTangent=-3.1312, interpMode=RCIM_Cubic, leaveTangent=-3.1312, tangentMode=RCTM_User, time=1, value=0}]}}}
      Scale Curve=0.1
     )
     1=&FloatFromCurve(
      FloatCurve=DataInterface() {propertyValues {Curve {
         defaultValue=1
         keys [
          {arriveTangent=9.0725, interpMode=RCIM_Cubic, leaveTangent=9.0725, tangentMode=RCTM_User, time=0, value=0}
          {arriveTangent=-2.5249, interpMode=RCIM_Cubic, leaveTangent=-2.5249, tangentMode=RCTM_User, time=0.1408, value=0.9847}
          {arriveTangent=-0.0405, interpMode=RCIM_Cubic, leaveTangent=-0.0405, tangentMode=RCTM_User, time=1, value=0}
         ]
        }}}
      Scale Curve=0.5
     )
     2=?
     3=?
     Index=$Particles.VisibilityTag
     Num=::NewEnumerator0
    )
    Scale RGB=&Multiply_Vector(
     A=&MakeVectorFromLinearColorRGB(Color=&ColorFromCurve(ColorCurve=DataInterface() {propertyValues {
         AlphaCurve {keys [<at 0 1>]}
         BlueCurve {keys [<at 0 0.2604>,<at 0.4157 0>,<at 1 0>]}
         GreenCurve {keys [<at 0 2.2352>,<at 0.4157 1.6667>,<at 1 0.25>]}
         RedCurve {keys [<at 0 5>,<at 0.4157 5>,<at 1 1>]}
        }}))
     B=<3 3 3>
    )
    ScaleA=true
    ScaleRGB=true
   )
   &ScaleMeshSize(Scale Factor=&VectorFromCurve(VectorCurve=DataInterface() {propertyValues {
       XCurve {
        defaultValue=1
        keys [
         {arriveTangent=1.7054, interpMode=RCIM_Cubic, leaveTangent=1.7054, tangentMode=RCTM_User, time=0, value=0}
         {arriveTangent=4.7155, tangentMode=RCTM_User, time=1, value=2}
        ]
       }
       YCurve {defaultValue=1, keys [{arriveTangent=2.2457, interpMode=RCIM_Cubic, leaveTangent=2.2457, tangentMode=RCTM_User, time=0, value=0}, <at 1 1>]}
       ZCurve {defaultValue=1, keys [{arriveTangent=2.2457, interpMode=RCIM_Cubic, leaveTangent=2.2457, tangentMode=RCTM_User, time=0, value=0}, <at 1 1>]}
      }}))
   &FN_Sprite_Attributes(
    Create SpriteSize Attribute=true
    SpriteSize=&Vector2DFromFloat(Value=&Lerp_Float(A=100, B=200, Alpha=$Particles.DistanceBlend))
    Write to Module Namespace=true
   ) {module=FN_Sprite}
   /SetVariables(Particles.SpriteColor=&MultiplyLinearColor(
     Linear Color=$Particles.Color
     Scale Factor=&MakeVector4(X=1, Y=1.3, Z=1, W=&Lerp_Float(A=0.5, B=1, Alpha=$Particles.DistanceBlend))
    )) {isSetParameters=true, module=SetVariables_#4}
   &ScaleSpriteSize(
    Uniform Curve Scale=2.5
    Uniform Curve Sprite Scale=DataInterface() {propertyValues {Curve {
       defaultValue=0
       keys [
        {arriveTangent=2.3161, interpMode=RCIM_Cubic, leaveTangent=2.3161, tangentMode=RCTM_User, time=0, value=0}
        {arriveTangent=0.0751, interpMode=RCIM_Cubic, leaveTangent=0.0751, tangentMode=RCTM_User, time=1, value=1}
       ]
      }}}
   )
   &FN_Light_Attributes(
    Color=<rgba 0.2708 0.125 1 1>
    Create Color Attribute=true
    Create Exponent Attribute=true
    Enable Particle Light Enable=true
    Exponent=4
    Light Enabled=true
    Scale Color=5
   )
   &FN_Radial_Material(
    Base Sprite Offset X:NiagaraFloat=&RandomRangeFloat(Minimum=-1, Maximum=1)
    Base Sprite Offset Y:NiagaraFloat=&RandomRangeFloat(Minimum=-1, Maximum=1)
    Sprite Extrusion Speed=0.5
    Write to Module Namespace=true
   )
   /SetVariables(
    Particles.Color_Glow:LinearColor=&MakeLinearColorFromVectorAndFloat(
     "Vector (RGB)"=&MakeVectorFromLinearColorRGB(Color=&ColorFromCurve(ColorCurve=DataInterface() {propertyValues {
         AlphaCurve {keys [<at 0 1>]}
         BlueCurve {keys [<at 0 0.2604>,<at 0.9949 0>]}
         GreenCurve {keys [<at 0 2.2352>,<at 0.9949 1.6667>]}
         RedCurve {keys [<at 0 5>,<at 0.9949 5>]}
        }}))
     "Float (Alpha)"=&FloatFromCurve(
      FloatCurve=DataInterface() {propertyValues {Curve {
         defaultValue=1
         keys [
          {arriveTangent=9.0725, interpMode=RCIM_Cubic, leaveTangent=9.0725, tangentMode=RCTM_User, time=0, value=0}
          {arriveTangent=-2.5249, interpMode=RCIM_Cubic, leaveTangent=-2.5249, tangentMode=RCTM_User, time=0.1408, value=0.9847}
          {arriveTangent=-0.0405, interpMode=RCIM_Cubic, leaveTangent=-0.0405, tangentMode=RCTM_User, time=1, value=0}
         ]
        }}}
      Scale Curve=0
     )
    )
    Particles.Size_Glow:Vector2f=&Multiply_Vector2D_ByFloat(Float=0.9, Vector2D=$Particles.SpriteSize)
   ) {isSetParameters=true, module=SetVariables_#5}
   &SolveForcesAndVelocity(Clamp Velocity=false, Limit Acceleration=false)
  ]
  rendererClasses [&NiagaraMeshRendererProperties, &NiagaraSpriteRendererProperties]
  renderers [
   {rendererClass=&NiagaraMeshRendererProperties, rendererIndex=0}
   {rendererClass=&NiagaraSpriteRendererProperties, rendererIndex=1}
   {rendererClass=&NiagaraSpriteRendererProperties, rendererIndex=2}
  ]
 }
 Flash001 {
  emitterUpdateScript [
   &EmitterState(
    Enable Distance Culling=true
    Life Cycle Mode=::NewEnumerator1
    Loop Behavior=::NewEnumerator1
    Loop Duration=1
    Max Distance Response=::KillAfterParticlesFinish
    MaxDistance=30000
    Module.bUseMaxDistance=true
    Module.bUseMinDistance=false
    Reset Age On Awaken=true
    Scalability Mode=::NewEnumerator1
    UseLoopDelay=false
   )
   &SpawnBurst_Instantaneous(Spawn Count=&RandomRangeInt(Minimum=3, Maximum=4), Use Spawn Probability=false)
   &SpawnBurst_Instantaneous(Spawn Group=1, Use Spawn Probability=false, UseLoopCountLimit=false) {module=SpawnBurst_Instantaneous001}
   &SpawnBurst_Instantaneous(Spawn Count=1, Spawn Group=2, Use Spawn Probability=false) {module=SpawnBurst_Instantaneous002}
  ]
  particleSpawnScript [
   /SetVariables(Particles.VisibilityTag=&SpawnGroup()) {isSetParameters=true, module=SetVariables_#1}
   &InitializeParticle(
    Lifetime=&MergeHelper_FloatSwitch(
     0=&RandomRangeFloat(Minimum=0.1, Maximum=0.125)
     1=0.2
     2=0.2
     3=?
     Index=$Particles.VisibilityTag
     Num=::NewEnumerator1
    )
    Mesh Scale=&UniformRangedVector(Minimum=<0.35 0.1 0.1>, Maximum=<0.5 0.2 0.2>)
    Mesh Scale Mode=::NewEnumerator4
    Sprite Size Mode=::NewEnumerator3
    Uniform Sprite Size=80
    UsePositionOffset=false
    Write Material Random=true
   ) {module=InitializeParticle001}
   &CameraOffset(Camera Offset Amount=10)
   &InitialMeshOrientation(
    Mesh Orientation Mode=::NewEnumerator2
    Orientation Coordinate Space=::Local
    Orientation Vector=<-2 0 -2>
    Rotation=&RandomRangeVector(Minimum=<-0.25 -0.25 0>, Maximum=<0.25 0.25 0>)
    Rotation Coordinate Space=::NewEnumerator0
    Use Rotation Vector=true
   ) {module=InitialMeshOrientation001}
   /SetVariables(Particles.LightPositionNew:Vector3f=&Add_Vector(
     A=Linked() {linkedVariable {name=Particles.Position, type=NiagaraPosition}}
     B=&TransformPosition(Position=<25 0 0>, Source Space=::Local, Destination Space=::Simulation)
    )) {isSetParameters=true}
   /SetVariables(Particles.LightExponent=3) {isSetParameters=true}
  ]
  particleUpdateScript [
   /SetVariables(Particles.DistanceBlend:NiagaraFloat=&FN_Fit(
     Clamp:NiagaraBool=true
     Value:NiagaraFloat=$Emitter.Scalability.DistanceFraction
     SourceMin:NiagaraFloat=0.1
     SourceMax:NiagaraFloat=1
     DestMin:NiagaraFloat=0
     DestMax:NiagaraFloat=1
    )) {isSetParameters=true, module=SetVariables_#3}
   &ParticleState()
   &ScaleColor(
    Scale Alpha=&MergeHelper_FloatSwitch(
     0=&FloatFromCurve(
      FloatCurve=DataInterface() {propertyValues {Curve {keys [<at 0 1 RCIM_Cubic>, {arriveTangent=-3.1312, interpMode=RCIM_Cubic, leaveTangent=-3.1312, tangentMode=RCTM_User, time=1, value=0}]}}}
      Scale Curve=0.1
     )
     1=&FloatFromCurve(
      FloatCurve=DataInterface() {propertyValues {Curve {
         defaultValue=1
         keys [
          {arriveTangent=9.0725, interpMode=RCIM_Cubic, leaveTangent=9.0725, tangentMode=RCTM_User, time=0, value=0}
          {arriveTangent=-2.5249, interpMode=RCIM_Cubic, leaveTangent=-2.5249, tangentMode=RCTM_User, time=0.1408, value=0.9847}
          {arriveTangent=-0.0405, interpMode=RCIM_Cubic, leaveTangent=-0.0405, tangentMode=RCTM_User, time=1, value=0}
         ]
        }}}
      Scale Curve=0.5
     )
     2=?
     3=?
     Index=$Particles.VisibilityTag
     Num=::NewEnumerator0
    )
    Scale RGB=&Multiply_Vector(
     A=&MakeVectorFromLinearColorRGB(Color=&ColorFromCurve(ColorCurve=DataInterface() {propertyValues {
         AlphaCurve {keys [<at 0 1>]}
         BlueCurve {keys [<at 0 0.2604>,<at 0.4157 0>,<at 1 0>]}
         GreenCurve {keys [<at 0 2.2352>,<at 0.4157 1.6667>,<at 1 0.25>]}
         RedCurve {keys [<at 0 5>,<at 0.4157 5>,<at 1 1>]}
        }}))
     B=<3 3 3>
    )
    ScaleA=true
    ScaleRGB=true
   )
   &ScaleMeshSize(Scale Factor=&VectorFromCurve(VectorCurve=DataInterface() {propertyValues {
       XCurve {
        defaultValue=1
        keys [
         {arriveTangent=1.7054, interpMode=RCIM_Cubic, leaveTangent=1.7054, tangentMode=RCTM_User, time=0, value=0}
         {arriveTangent=4.7155, tangentMode=RCTM_User, time=1, value=2}
        ]
       }
       YCurve {defaultValue=1, keys [{arriveTangent=2.2457, interpMode=RCIM_Cubic, leaveTangent=2.2457, tangentMode=RCTM_User, time=0, value=0}, <at 1 1>]}
       ZCurve {defaultValue=1, keys [{arriveTangent=2.2457, interpMode=RCIM_Cubic, leaveTangent=2.2457, tangentMode=RCTM_User, time=0, value=0}, <at 1 1>]}
      }}))
   &FN_Sprite_Attributes(
    Create SpriteSize Attribute=true
    SpriteSize=&Vector2DFromFloat(Value=&Lerp_Float(A=100, B=200, Alpha=$Particles.DistanceBlend))
    Write to Module Namespace=true
   ) {module=FN_Sprite}
   /SetVariables(Particles.SpriteColor=&MultiplyLinearColor(
     Linear Color=$Particles.Color
     Scale Factor=&MakeVector4(X=1, Y=1.3, Z=1, W=&Lerp_Float(A=0.5, B=1, Alpha=$Particles.DistanceBlend))
    )) {isSetParameters=true, module=SetVariables_#4}
   &ScaleSpriteSize(
    Uniform Curve Scale=2.5
    Uniform Curve Sprite Scale=DataInterface() {propertyValues {Curve {
       defaultValue=0
       keys [
        {arriveTangent=2.3161, interpMode=RCIM_Cubic, leaveTangent=2.3161, tangentMode=RCTM_User, time=0, value=0}
        {arriveTangent=0.0751, interpMode=RCIM_Cubic, leaveTangent=0.0751, tangentMode=RCTM_User, time=1, value=1}
       ]
      }}}
   )
   &FN_Radial_Material(
    Base Sprite Offset X:NiagaraFloat=&RandomRangeFloat(Minimum=-1, Maximum=1)
    Base Sprite Offset Y:NiagaraFloat=&RandomRangeFloat(Minimum=-1, Maximum=1)
    Sprite Extrusion Speed=0.5
    Write to Module Namespace=true
   )
   /SetVariables(
    Particles.Color_Glow:LinearColor=&MakeLinearColorFromVectorAndFloat(
     "Vector (RGB)"=&MakeVectorFromLinearColorRGB(Color=&ColorFromCurve(ColorCurve=DataInterface() {propertyValues {
         AlphaCurve {keys [<at 0 1>]}
         BlueCurve {keys [<at 0 0.2604>,<at 0.9949 0>]}
         GreenCurve {keys [<at 0 2.2352>,<at 0.9949 1.6667>]}
         RedCurve {keys [<at 0 5>,<at 0.9949 5>]}
        }}))
     "Float (Alpha)"=&FloatFromCurve(
      FloatCurve=DataInterface() {propertyValues {Curve {
         defaultValue=1
         keys [
          {arriveTangent=9.0725, interpMode=RCIM_Cubic, leaveTangent=9.0725, tangentMode=RCTM_User, time=0, value=0}
          {arriveTangent=-2.5249, interpMode=RCIM_Cubic, leaveTangent=-2.5249, tangentMode=RCTM_User, time=0.1408, value=0.9847}
          {arriveTangent=-0.0405, interpMode=RCIM_Cubic, leaveTangent=-0.0405, tangentMode=RCTM_User, time=1, value=0}
         ]
        }}}
      Scale Curve=0
     )
    )
    Particles.Size_Glow:Vector2f=&Multiply_Vector2D_ByFloat(Float=0.9, Vector2D=$Particles.SpriteSize)
   ) {isSetParameters=true, module=SetVariables_#5}
   /SetVariables(Particles.Light_Color:LinearColor=&MultiplyLinearColor(
     Linear Color=<rgba 1 0.574 0.1481 1>
     Scale Factor=&MakeVector4_FromFloat(Float=&Multiply_Float(
       A=20
       B=&FloatFromCurve(
        FloatCurve=DataInterface() {propertyValues {Curve {keys [
            {interpMode=RCIM_Cubic, tangentMode=RCTM_User, time=0, value=0}
            {interpMode=RCIM_Cubic, tangentMode=RCTM_User, time=0.2, value=1}
            {tangentMode=RCTM_User, time=1, value=0}
           ]}}}
        Scale Curve=/Niagara/DynamicInputs/Scalability/ScalabilityDistanceBasedFloat(
         Curve:NiagaraDataInterfaceCurve=DataInterface() {propertyValues {Curve {keys [
             {interpMode=RCIM_Cubic, tangentMode=RCTM_Break, time=0, value=1}
             {arriveTangent=0.0625, interpMode=RCIM_Cubic, tangentMode=RCTM_Break, time=1, value=0}
            ]}}}
         CurveIndex:NiagaraFloat=$Emitter.Scalability.DistanceFraction
         Bias:NiagaraFloat=0
         Scale:NiagaraFloat=1
        )
       )
      ))
    )) {isSetParameters=true}
   /SetVariables(Particles.LightRadius=&Multiply_Float(
     A=40
     B=&ClampFloat(
      Float=&FN_LOD_Distance_Remap(Start Distance=0, End Distance=8000, Minimum Output=1, Maximum Output=2)
      Min=1
      Max=2
     )
    )) {isSetParameters=true}
   &SolveForcesAndVelocity(Clamp Velocity=false, Limit Acceleration=false)
  ]
  rendererClasses [
   &NiagaraMeshRendererProperties
   &NiagaraSpriteRendererProperties
   &NiagaraLightRendererProperties
  ]
  renderers [
   {rendererClass=&NiagaraMeshRendererProperties, rendererIndex=0}
   {rendererClass=&NiagaraSpriteRendererProperties, rendererIndex=1}
   {rendererClass=&NiagaraLightRendererProperties, rendererIndex=2}
  ]
 }
 Light001 {
  bEnabled=false
  emitterUpdateScript [
   &EmitterState(
    Enable Distance Culling=true
    Life Cycle Mode=::NewEnumerator1
    Loop Behavior=::NewEnumerator1
    Loop Duration=1
    MaxDistance=4000
    Module.bUseMaxDistance=true
    Module.bUseMinDistance=true
    Scalability Mode=::NewEnumerator1
    UseLoopDelay=false
   )
   &SpawnBurst_Instantaneous(Spawn Count=1, Use Spawn Probability=false)
  ]
  particleSpawnScript [
   &InitializeParticle(Lifetime=0.2, UsePositionOffset=false, Write Material Random=true) {module=InitializeParticle001}
   /SetVariables(Particles.LightPositionNew:Vector3f=&Add_Vector(
     A=Linked() {linkedVariable {name=Particles.Position, type=NiagaraPosition}}
     B=&TransformPosition(Position=<25 0 0>, Source Space=::Local, Destination Space=::Simulation)
    )) {isSetParameters=true, module=SetVariables_#2}
   /SetVariables(Particles.LightExponent=3) {isSetParameters=true}
  ]
  particleUpdateScript [
   /SetVariables(Particles.Light_Color:LinearColor=&MultiplyLinearColor(
     Linear Color=<rgba 1 0.574 0.1481 1>
     Scale Factor=&MakeVector4_FromFloat(Float=&Multiply_Float(
       A=20
       B=&FloatFromCurve(
        FloatCurve=DataInterface() {propertyValues {Curve {keys [
            {interpMode=RCIM_Cubic, tangentMode=RCTM_User, time=0, value=0}
            {interpMode=RCIM_Cubic, tangentMode=RCTM_User, time=0.2, value=1}
            {tangentMode=RCTM_User, time=1, value=0}
           ]}}}
        Scale Curve=&ScalabilityDistanceBasedFloat(
         Curve:NiagaraDataInterfaceCurve=DataInterface() {propertyValues {Curve {keys [
             {interpMode=RCIM_Cubic, tangentMode=RCTM_Break, time=0, value=1}
             {arriveTangent=0.0625, interpMode=RCIM_Cubic, tangentMode=RCTM_Break, time=1, value=0}
            ]}}}
         CurveIndex:NiagaraFloat=$Emitter.Scalability.DistanceFraction
         Bias:NiagaraFloat=0
         Scale:NiagaraFloat=1
        )
       )
      ))
    )) {isSetParameters=true}
   /SetVariables(Particles.LightRadius=&Multiply_Float(
     A=40
     B=&ClampFloat(
      Float=&FN_LOD_Distance_Remap(Start Distance=0, End Distance=8000, Minimum Output=1, Maximum Output=2)
      Min=1
      Max=2
     )
    )) {isSetParameters=true}
   &ParticleState()
  ]
  rendererClasses [&NiagaraLightRendererProperties]
  renderers [{rendererClass=&NiagaraLightRendererProperties, rendererIndex=0}]
 }
 Sparks {
  emitterUpdateScript [
   &EmitterState(Life Cycle Mode=::NewEnumerator1, Loop Behavior=::NewEnumerator1, Loop Delay=0.05, Loop Duration=1, UseLoopDelay=false)
   &SpawnBurst_Instantaneous(Spawn Count=&RandomRangeInt(Minimum=20, Maximum=35), Use Spawn Probability=false)
  ]
  particleSpawnScript [
   /SetVariables(Particles.ParticleRandom:NiagaraFloat=&SelectFloatFromArray(
     Direct Array Index=/Niagara/DynamicInputs/Arrays/SelectIntFromWeightedDistributionArray(
      Array Sampling Mode:ENiagara_LifetimeMode=::NewEnumerator1
      Direct Array Index:NiagaraInt32=?
      Direct Array Mode:ENiagara_WrapClamp=::NewEnumerator0
      Evaluation Type:ENiagaraRandomnessEvaluation=::NewEnumerator0
      Recalculate Random Each Loop:NiagaraBool=false
      Weighted Distribution Int Array:NiagaraDataInterfaceArrayDistributionInt=DataInterface() {propertyValues {ArrayData [{value=0, weight=0.65}, {value=1, weight=0.35}]}}
      Randomness Mode:ENiagaraRandomnessMode=::NewEnumerator0
      Random Seed:NiagaraInt32=0
      Fixed Random Seed:NiagaraBool=false
     )
     Float Selection Array=DataInterface() {propertyValues {FloatData [0,1]}}
    )) {isSetParameters=true, module=SetVariables_#6}
   &InitializeParticle(
    Color=<rgba 20 12.3188 1.5652 1>
    Lifetime=&MergeHelper_FloatSwitch(
     0=&UniformRangedFloat(Minimum=0.15, Maximum=0.5)
     1=&RandomRangeFloat(Minimum=1, Maximum=1.5)
     2=?
     3=?
     Index=&MakeIntFromFloat(Float=$Particles.ParticleRandom, Conversion Mode=::NewEnumerator1)
     Num=::NewEnumerator0
    )
    Sprite Size Max=<1 8>
    Sprite Size Min=<0.5 3>
    Sprite Size Mode=::NewEnumerator2
    UsePositionOffset=false
    Write Material Random=true
   ) {module=InitializeParticle001}
   &ShapeLocation(MaskSpawnGroup=false, Override Seed=false, Sphere Radius=4, UseOwnerScale=true)
   &AddVelocity(
    Cone Angle=120
    Cone Axis=<-1 0.5 -1>
    Inner Cone Angle=60
    Module.UseConeFalloff=false
    Module.UseVelDistribution=false
    Override Seed=false
    Velocity Mode=::NewEnumerator2
    Velocity Speed=&RandomRangeFloat(Minimum=20, Maximum=350)
   ) {module=AddVelocity001}
   &AddVelocity(
    Module.Use Falloff Distance=false
    Override Seed=false
    Velocity Mode=::NewEnumerator1
    Velocity Speed=&RandomRangeFloat(Minimum=25, Maximum=250)
   ) {enabled=false}
  ]
  particleUpdateScript [
   &Drag(Drag=&RandomRangeFloat(Minimum=0.2, Maximum=1), Use Linear Drag=true, Use Rotational Drag=false)
   &ScaleColor(
    Scale Alpha=&FloatFromCurve(FloatCurve=DataInterface() {propertyValues {Curve {keys [
         {arriveTangent=1.1765, interpMode=RCIM_Cubic, time=0.1234, value=0}
         {arriveTangent=2.6551, interpMode=RCIM_Cubic, leaveTangent=2.6551, time=0.2484, value=0.7647}
         {arriveTangent=-1.0174, interpMode=RCIM_Cubic, leaveTangent=-1.0174, time=0.5, value=1}
         {arriveTangent=-4.9904, interpMode=RCIM_Cubic, leaveTangent=-4.9904, tangentMode=RCTM_User, time=1, value=0}
        ]}}})
    Scale RGB=&MakeVectorFromLinearColorRGB(Color=&ColorFromCurve(ColorCurve=DataInterface() {propertyValues {
        AlphaCurve {keys [<at 0 1>]}
        BlueCurve {keys [<at 0 0.2604>,<at 0.1639 0>,<at 1 0>]}
        GreenCurve {keys [<at 0 1.0503>,<at 0.1639 1.75>,<at 1 0.25>]}
        RedCurve {keys [<at 0 5>,<at 0.1639 7>,<at 1 1>]}
       }}))
    ScaleA=true
    ScaleRGB=true
   ) {module=ScaleColor001}
   &GravityForce(Gravity=&MergeHelper_VectorSwitch(
     0=&RandomRangeVector(Minimum=<0 0 -150>, Maximum=<0 0 -500>)
     1=&RandomRangeVector(Minimum=<0 0 -500>, Maximum=<0 0 -980>)
     2=?
     3=?
     Index=&MakeIntFromFloat(Float=$Particles.ParticleRandom, Conversion Mode=::NewEnumerator1)
     Num=::NewEnumerator0
    ))
   &Collision(
    Advanced Aging Rate=1
    Age Colliding Particles=false
    Collision Enabled=&SetBoolByFloatComparison(
     A=&MakeCustomFloatFromBool(
      Bool=&IsPlatformSetActive(New Platform Set=DataInterface() {propertyValues {Platforms {qualityLevelMask=-4}}})
      True Float=&MakeCustomFloatFromBool(
       Bool=&SetBoolByFloatComparison(A=$Particles.ParticleRandom, B=1, Comparison Type=::NewEnumerator4)
       True Float=1
       False Float=0
      )
      False Float=0
     )
     B=0
    )
    Enable Rest State=false
    Friction=0.6
    Friction During a Bounce=0.4
    Kill On Collision:NiagaraInt32=&RandomRangeInt(Minimum=3, Maximum=5)
    KillOnCollision=true
    Module.EnableMaxCollisionCount=false
    Module.Randomize Collision Normal Vector=false
   )
   &SolveForcesAndVelocity(Clamp Velocity=false, Limit Acceleration=false)
   &ParticleState()
   &ScaleSpriteSize(
    Scale Sprite Size Mode=::NewEnumerator0
    Uniform Scale Factor=&Lerp_Float(
     A=1.5
     B=&Multiply_Float(A=&Waveform("Waveform [1]"=::NewEnumerator0, "[1] Frequency"=75, "[1] Amplitude Min/Max"=<0.5 1>), B=1.5)
     Alpha=&FloatFromCurve(FloatCurve=DataInterface() {propertyValues {Curve {keys [<at 0 0.1076>,<at 0.2485 0.6128>,<at 0.45 1>,<at 1 1>]}}})
    )
   )
   &ScaleSpriteSizeBySpeed(Max Scale Factor=<1 3>, Min Scale Factor=<0.5 0.5>)
  ]
  rendererClasses [&NiagaraSpriteRendererProperties]
  renderers [{rendererClass=&NiagaraSpriteRendererProperties}]
 }
 Sparks001 {
  bEnabled=false
  emitterUpdateScript [
   &EmitterState(Life Cycle Mode=::NewEnumerator1, Loop Behavior=::NewEnumerator1, Loop Delay=0.05, Loop Duration=1, UseLoopDelay=false)
   &SpawnBurst_Instantaneous(Spawn Count=&RandomRangeInt(Minimum=20, Maximum=35), Use Spawn Probability=false)
  ]
  particleSpawnScript [
   /SetVariables(Particles.ParticleRandom:NiagaraFloat=&SelectFloatFromArray(
     Direct Array Index=&SelectIntFromWeightedDistributionArray(
      Array Sampling Mode:ENiagara_LifetimeMode=::NewEnumerator1
      Direct Array Index:NiagaraInt32=?
      Direct Array Mode:ENiagara_WrapClamp=::NewEnumerator0
      Evaluation Type:ENiagaraRandomnessEvaluation=::NewEnumerator0
      Recalculate Random Each Loop:NiagaraBool=false
      Weighted Distribution Int Array:NiagaraDataInterfaceArrayDistributionInt=DataInterface() {propertyValues {ArrayData [{value=0, weight=0.75}, {value=1, weight=0.35}]}}
      Randomness Mode:ENiagaraRandomnessMode=::NewEnumerator0
      Random Seed:NiagaraInt32=0
      Fixed Random Seed:NiagaraBool=false
     )
     Float Selection Array=DataInterface() {propertyValues {FloatData [0,1]}}
    )) {isSetParameters=true, module=SetVariables_#6}
   &InitializeParticle(
    Color=<rgba 20 12.3188 1.5652 1>
    Lifetime=&MergeHelper_FloatSwitch(
     0=&UniformRangedFloat(Minimum=0.15, Maximum=0.5)
     1=&RandomRangeFloat(Minimum=1, Maximum=1.5)
     2=?
     3=?
     Index=&MakeIntFromFloat(Float=$Particles.ParticleRandom, Conversion Mode=::NewEnumerator1)
     Num=::NewEnumerator0
    )
    Sprite Size Max=<1 8>
    Sprite Size Min=<0.5 3>
    Sprite Size Mode=::NewEnumerator2
    UsePositionOffset=false
    Write Material Random=true
   ) {module=InitializeParticle001}
   &ShapeLocation(MaskSpawnGroup=false, Override Seed=false, Sphere Radius=4, UseOwnerScale=true)
   &AddVelocity(
    Cone Angle=120
    Cone Axis=<-1 0.5 -1>
    Inner Cone Angle=60
    Module.UseConeFalloff=false
    Module.UseVelDistribution=false
    Override Seed=false
    Velocity Mode=::NewEnumerator2
    Velocity Speed=&RandomRangeFloat(Minimum=20, Maximum=350)
   ) {module=AddVelocity001}
   &AddVelocity(
    Module.Use Falloff Distance=false
    Override Seed=false
    Velocity Mode=::NewEnumerator1
    Velocity Speed=&RandomRangeFloat(Minimum=25, Maximum=250)
   ) {enabled=false}
  ]
  particleUpdateScript [
   &Drag(Drag=&RandomRangeFloat(Minimum=0.2, Maximum=1), Use Linear Drag=true, Use Rotational Drag=false)
   &ScaleColor(
    Scale Alpha=&FloatFromCurve(FloatCurve=DataInterface() {propertyValues {Curve {keys [
         {arriveTangent=1.1765, interpMode=RCIM_Cubic, time=0.1234, value=0}
         {arriveTangent=2.6551, interpMode=RCIM_Cubic, leaveTangent=2.6551, time=0.2484, value=0.7647}
         {arriveTangent=-1.0174, interpMode=RCIM_Cubic, leaveTangent=-1.0174, time=0.5, value=1}
         {arriveTangent=-4.9904, interpMode=RCIM_Cubic, leaveTangent=-4.9904, tangentMode=RCTM_User, time=1, value=0}
        ]}}})
    Scale RGB=&MakeVectorFromLinearColorRGB(Color=&ColorFromCurve(ColorCurve=DataInterface() {propertyValues {
        AlphaCurve {keys [<at 0 1>]}
        BlueCurve {keys [<at 0 0.2604>,<at 0.1639 0>,<at 1 0>]}
        GreenCurve {keys [<at 0 1.0503>,<at 0.1639 1.75>,<at 1 0.25>]}
        RedCurve {keys [<at 0 5>,<at 0.1639 7>,<at 1 1>]}
       }}))
    ScaleA=true
    ScaleRGB=true
   ) {module=ScaleColor001}
   &GravityForce(Gravity=&MergeHelper_VectorSwitch(
     0=&RandomRangeVector(Minimum=<0 0 -150>, Maximum=<0 0 -500>)
     1=&RandomRangeVector(Minimum=<0 0 -500>, Maximum=<0 0 -980>)
     2=?
     3=?
     Index=&MakeIntFromFloat(Float=$Particles.ParticleRandom, Conversion Mode=::NewEnumerator1)
     Num=::NewEnumerator0
    ))
   &Collision(
    Advanced Aging Rate=1
    Age Colliding Particles=false
    Collision Enabled=&SetBoolByFloatComparison(A=$Particles.ParticleRandom, B=1, Comparison Type=::NewEnumerator4)
    Enable Rest State=false
    Friction=0.6
    Friction During a Bounce=0.4
    Kill On Collision:NiagaraInt32=&RandomRangeInt(Minimum=3, Maximum=5)
    KillOnCollision=true
    Module.EnableMaxCollisionCount=false
    Module.Randomize Collision Normal Vector=false
   )
   &SolveForcesAndVelocity(Clamp Velocity=false, Limit Acceleration=false)
   &ParticleState()
   &ScaleSpriteSize(
    Scale Sprite Size Mode=::NewEnumerator0
    Uniform Scale Factor=&Lerp_Float(
     A=1.5
     B=&Multiply_Float(A=&Waveform("Waveform [1]"=::NewEnumerator0, "[1] Frequency"=75, "[1] Amplitude Min/Max"=<0.5 1>), B=1.5)
     Alpha=&FloatFromCurve(FloatCurve=DataInterface() {propertyValues {Curve {keys [<at 0 0.1076>,<at 0.2485 0.6128>,<at 0.45 1>,<at 1 1>]}}})
    )
   )
   &ScaleSpriteSizeBySpeed(Max Scale Factor=<1 3>, Min Scale Factor=<0.5 0.5>)
  ]
  rendererClasses [&NiagaraSpriteRendererProperties]
  renderers [{rendererClass=&NiagaraSpriteRendererProperties}]
 }
}
summary {
 systemName=NS_Emote_NobleCrusher_SwordGroundContact
 emitters [
  {bEnabled=false, emitterName=Sparks001, rendererClasses [&NiagaraSpriteRendererProperties]}
  {bEnabled=false, emitterName=Light001, rendererClasses [&NiagaraLightRendererProperties]}
  {
   bEnabled=false
   emitterName=Flash
   rendererClasses [&NiagaraMeshRendererProperties, &NiagaraSpriteRendererProperties]
  }
  {emitterName=Sparks, rendererClasses [&NiagaraSpriteRendererProperties]}
  {
   emitterName=Flash001
   rendererClasses [
    &NiagaraMeshRendererProperties
    &NiagaraSpriteRendererProperties
    &NiagaraLightRendererProperties
   ]
  }
 ]
 userVariables [User.Multiplier {type=NiagaraFloat, defaultValue=1, description=""}]
}
